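# Control Rig for the M_MED_ButtonCase player body that drives corrective deformation of its skirt, belt, collar and chain bones.
# A shared DeformRig function runs first, then weighted parent constraints place cloth and accessory bones.
# The Belt and Skirt Setup drives offsets from thigh pose readers, calf bend and thigh aim controls, using capsule clamps against penetration.
# Emote, DBNO, skydiving, sliding and knee-sliding flags switch or blend corrections.
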ControlRigBlueprint /BRCosmetics/Characters/Player/Male/Medium/Bodies/M_MED_ButtonCase/Meshes/M_MED_ButtonCase_CtrlRig
graphs [
 New Function_ContainedGraph {nodes [
   RigVMFunctionEntryNode()
   > RigVMFunction_Sequence = RigVMAggregateNode()
   RigVMFunction_MathBoolOr = RigVMAggregateNode(
    A=RigVMVariableNode() {variable=IsEmote}.Value
    B=RigVMVariableNode() {variable=DBNO}.Value
    C=RigVMVariableNode() {variable=Skydiving}.Value
   )
   RigVMFunction_ControlFlowBranch = &RigVMFunction_ControlFlowBranch(Condition=@RigVMFunction_MathBoolOr.Result, ExecuteContext=@RigVMFunction_Sequence.C)
   n1 = &RigUnit_SphericalPoseReader(
    ActiveRegionScaleFactors=(PositiveWidth=3.5,NegativeWidth=0.5,PositiveHeight=2,NegativeHeight=2)
    ActiveRegionSize=0.1
    Debug=(bDrawDebug=false,DebugScale=25,DebugSegments=20,DebugThickness=0.25)
    DriverAxis=(X=-1,Y=0,Z=0)
    DriverItem=(Type=Bone,Name="thigh_l")
    FalloffRegionScaleFactors=(PositiveWidth=1,NegativeWidth=1,PositiveHeight=1,NegativeHeight=1)
    FalloffSize=0.45
    OptionalParentItem=(Type=Bone)
    RotationOffset=(X=90,Y=0,Z=-30)
    ExecutePin=@RigVMFunction_ControlFlowBranch.False
   )
   > n2 = &RigUnit_SphericalPoseReader(
    ActiveRegionScaleFactors=(PositiveWidth=3.5,NegativeWidth=0.5,PositiveHeight=2,NegativeHeight=2)
    ActiveRegionSize=0.1
    Debug=(bDrawDebug=false,DebugScale=25,DebugSegments=20,DebugThickness=0.25)
    DriverAxis=(X=1,Y=0,Z=0)
    DriverItem=(Type=Bone,Name="thigh_r")
    FalloffRegionScaleFactors=(PositiveWidth=1,NegativeWidth=1,PositiveHeight=1,NegativeHeight=1)
    FalloffSize=0.45
    FlipHeightScaling=false
    FlipWidthScaling=true
    OptionalParentItem=(Type=Bone)
    RotationOffset=(X=-90,Y=0,Z=-30)
   )
   n3 = &RigUnit_GetRelativeTransformForItem(Child=(Type=Bone,Name="calf_r"), Parent=(Type=Bone,Name="calf_r"), bChildInitial=False, bParentInitial=true)
   Subtract = &RigVMFunction_MathDoubleSub(
    A=&RigVMFunction_MathDoubleRemap(
     SourceMaximum=100
     SourceMinimum=10
     TargetMaximum=1
     TargetMinimum=0
     Value=&RigVMFunction_MathQuaternionToEuler(RotationOrder=ZYX, Value=@n3.RelativeTransform.Rotation).Result.Z
    ).Result
    B=@n2.OutputParam
   )
   > &RigUnit_ModifyTransforms(
    ItemToModify="((Item=(Type=Bone,Name=\"dyn_skirt_bk_1_r\"),Transform=(Rotation=(X=0,Y=0,Z=-0.1736,W=0.9848),Translation=(X=0,Y=0,Z=0),Scale3D=(X=1,Y=1,Z=1))),(Item=(Type=Bone,Name=\"dyn_skirt_bk_1_l\"),Transform=(Rotation=(X=0,Y=0,Z=-0.0872,W=0.9962),Translation=(X=0,Y=0,Z= ...(truncated)"
    WeightMinimum=0
    Weight=@Subtract.Result
   )
   RigUnit_GetRelativeTransformForItem = &RigUnit_GetRelativeTransformForItem(Child=(Type=Bone,Name="calf_l"), Parent=(Type=Bone,Name="calf_l"), bChildInitial=False, bParentInitial=true)
   RigVMFunction_MathQuaternionToEuler = &RigVMFunction_MathQuaternionToEuler(RotationOrder=ZYX, Value=@RigUnit_GetRelativeTransformForItem.RelativeTransform.Rotation)
   Remap = &RigVMFunction_MathDoubleRemap(SourceMaximum=100, SourceMinimum=10, TargetMaximum=1, TargetMinimum=0, Value=@RigVMFunction_MathQuaternionToEuler.Result.Z)
   > &RigUnit_ModifyTransforms(
    ItemToModify="((Item=(Type=Bone,Name=\"dyn_skirt_bk_1_l\"),Transform=(Rotation=(X=0,Y=0,Z=-0.1736,W=0.9848),Translation=(X=0,Y=0,Z=0),Scale3D=(X=1,Y=1,Z=1))),(Item=(Type=Bone,Name=\"dyn_skirt_bk_1_r\"),Transform=(Rotation=(X=0,Y=0,Z=-0.0872,W=0.9962),Translation=(X=0,Y=0,Z= ...(truncated)"
    WeightMinimum=0
    Weight=&RigVMFunction_MathDoubleSub(A=@Remap.Result, B=@n1.OutputParam).Result
   )
   n4 = &RigUnit_SphericalPoseReader(
    ActiveRegionScaleFactors=(PositiveWidth=3.5,NegativeWidth=0.5,PositiveHeight=2,NegativeHeight=2)
    ActiveRegionSize=0.1
    Debug=(bDrawDebug=false,DebugScale=25,DebugSegments=20,DebugThickness=0.25)
    DriverAxis=(X=-1,Y=0,Z=0)
    DriverItem=(Type=Bone,Name="thigh_l")
    FalloffRegionScaleFactors=(PositiveWidth=1,NegativeWidth=1,PositiveHeight=1,NegativeHeight=1)
    FalloffSize=0.45
    OptionalParentItem=(Type=Bone)
    RotationOffset=(X=90,Y=0,Z=-30)
    ExecutePin=@RigVMFunction_Sequence.D
   )
   > n5 = &RigUnit_SphericalPoseReader(
    ActiveRegionScaleFactors=(PositiveWidth=3.5,NegativeWidth=0.5,PositiveHeight=2,NegativeHeight=2)
    ActiveRegionSize=0.1
    Debug=(bDrawDebug=false,DebugScale=25,DebugSegments=20,DebugThickness=0.25)
    DriverAxis=(X=1,Y=0,Z=0)
    DriverItem=(Type=Bone,Name="thigh_r")
    FalloffRegionScaleFactors=(PositiveWidth=1,NegativeWidth=1,PositiveHeight=1,NegativeHeight=1)
    FalloffSize=0.45
    FlipHeightScaling=false
    FlipWidthScaling=true
    OptionalParentItem=(Type=Bone)
    RotationOffset=(X=-90,Y=0,Z=-30)
   )
   Add = &RigVMFunction_MathDoubleAdd(B=@n4.OutputParam, A=@n5.OutputParam)
   > &RigUnit_ModifyTransforms(
    ItemToModify="((Item=(Type=Bone,Name=\"deform_belt\"),Transform=(Rotation=(X=0,Y=0,Z=0.2588,W=0.9659),Translation=(X=13,Y=0,Z=0),Scale3D=(X=1.1,Y=1,Z=1))),(Item=(Type=Bone,Name=\"deform_belt1\"),Transform=(Rotation=(X=0,Y=0,Z=-0.0872,W=0.9962),Translation=(X=0,Y=-3.5,Z=0),S ...(truncated)"
    WeightMinimum=0
    Weight=&RigVMFunction_MathDoubleClamp(Value=@Add.Result).Result
   )
   > &RigUnit_ModifyTransforms(
    ItemToModify=((Item=(Type=Bone,Name="deform_belt"),Transform=(Rotation=(X=0,Y=-0.4226,Z=0,W=0.9063),Translation=(X=0,Y=0,Z=0),Scale3D=(X=1,Y=1,Z=1))))
    WeightMinimum=0
    Weight=@n4.OutputParam
   )
   > &RigUnit_ModifyTransforms(
    ItemToModify=((Item=(Type=Bone,Name="deform_belt"),Transform=(Rotation=(X=0,Y=0.4226,Z=0,W=0.9063),Translation=(X=0,Y=0,Z=0),Scale3D=(X=1,Y=1,Z=1))))
    WeightMinimum=0
    Weight=@n5.OutputParam
   )
   RigVMFunction_MathBoolToFloat = &RigVMFunction_MathBoolToFloat(Value=RigVMVariableNode() {variable=Sliding}.Value)
   AlphaInterp = &RigVMFunction_AlphaInterp(InterpSpeedDecreasing=10, InterpSpeedIncreasing=10, Scale=1, bClampResult=False, Value=@RigVMFunction_MathBoolToFloat.Result)
   > &RigUnit_ModifyTransforms(
    ItemToModify="((Item=(Type=Bone,Name=\"deform_skirt_bk_l\"),Transform=(Rotation=(X=-0.0644,Y=0.0778,Z=-0.7019,W=0.7051),Translation=(X=0,Y=0,Z=0),Scale3D=(X=1,Y=1,Z=1))),(Item=(Type=Bone,Name=\"deform_skirt_bk_r\"),Transform=(Rotation=(X=-0.0906,Y=0.1707,Z=-0.8247,W=0.5316) ...(truncated)"
    WeightMinimum=0
    Weight=@AlphaInterp.Result
   )
   n6 = &RigVMFunction_AlphaInterp(
    InterpSpeedDecreasing=10
    InterpSpeedIncreasing=15
    Scale=1
    bClampResult=False
    Value=&RigVMFunction_MathBoolToFloat(Value=RigVMVariableNode() {variable=KneeSliding}.Value).Result
   )
   > &RigUnit_ModifyTransforms(
    ItemToModify="((Item=(Type=Bone,Name=\"deform_skirt_bk_l\"),Transform=(Rotation=(X=-0.1865,Y=-0.1206,Z=-0.1921,W=0.9559),Translation=(X=-0.3256,Y=-2.6732,Z=2.2452),Scale3D=(X=1,Y=1,Z=1))),(Item=(Type=Bone,Name=\"deform_skirt_bk_r\"),Transform=(Rotation=(X=0.3529,Y=0.0219,Z= ...(truncated)"
    WeightMinimum=0
    Weight=@n6.Result
   )
   > &RigUnit_AimConstraintLocalSpaceOffset(
    Child=(Type=Bone,Name="deform_chain_up_out_l")
    Filter=(bX=True,bY=True,bZ=True)
    Parents=((Item=(Type=Bone,Name="dyn_chain_up_inn_1_l"),Weight=1))
    UpAxis=(X=0,Y=-1,Z=0)
    Weight=1
    WorldUp=(Target=(X=0,Y=0,Z=1),Kind=Location,Space=(Type=Bone,Name="deform_chain_up_out_l"))
    bMaintainOffset=true
   )
   > &RigUnit_AimConstraintLocalSpaceOffset(
    Child=(Type=Bone,Name="deform_chain_down_inn_l")
    Filter=(bX=True,bY=True,bZ=True)
    Parents=((Item=(Type=Bone,Name="dyn_chain_down_inn_2_l"),Weight=1))
    UpAxis=(X=0,Y=-1,Z=0)
    Weight=1
    WorldUp=(Target=(X=0,Y=0,Z=1),Kind=Location,Space=(Type=Bone,Name="deform_chain_down_inn_l"))
   )
   > &RigUnit_AimConstraintLocalSpaceOffset(
    Child=(Type=Bone,Name="deform_chain_up_out_r")
    Filter=(bX=True,bY=True,bZ=True)
    Parents=((Item=(Type=Bone,Name="dyn_chain_up_inn_1_r"),Weight=1))
    UpAxis=(X=0,Y=-1,Z=0)
    Weight=1
    WorldUp=(Target=(X=0,Y=0,Z=1),Kind=Location,Space=(Type=Bone,Name="deform_chain_up_out_r"))
    bMaintainOffset=true
   )
   > Set Transform_2 = /Script/ControlRig.RigUnit_SetTranslation(
    Item=(Type=Bone,Name="dyn_chain_up_out_1_r")
    Space=GlobalSpace
    Weight=1
    bInitial=false
    bPropagateToChildren=true
    Value=/Script/RigVM.RigVMFunction_MathTransformClampSpatially(
     Axis=X
     DebugColor=(R=1,G=0,B=0,A=1)
     DebugThickness=1
     Maximum=50
     Minimum=10
     Type=Capsule
     bDrawDebug=true
     Value=&RigUnit_GetTransform(Item=(Type=Bone,Name="dyn_chain_down_out_1_r"), bInitial=false).Transform
     Space=&RigUnit_GetTransform(Item=(Type=Bone,Name="upperarm_twist_01_r"), bInitial=false).Transform
    ) {
     is=struct RigVMUnitNode
     template="ClampSpatially::Execute(in Value,in Axis,in Type,in Minimum,in Maximum,in Space,in bDrawDebug,in DebugColor,in DebugThickness,out Result)"
    }.Result.Translation
   ) {is=struct RigVMUnitNode, template="Set Transform::Execute(in Item,in Space,in bInitial,in Value,in Weight,in bPropagateToChildren)"}
   RigVMFunctionReturnNode(ExecuteContext=@RigVMFunction_Sequence.E)
   &RigUnit_ParentConstraint(
    AdvancedSettings=(RotationOrderForFilter=XZY)
    Child=(Type=Bone,Name="deform_skirt_f_l")
    Parents="((Item=(Type=Bone,Name=\"thigh_twist_01_r\"),Weight=0.0003),(Item=(Type=Bone,Name=\"spine_03\"),Weight=0.022),(Item=(Type=Bone,Name=\"spine_01\"),Weight=0.0337),(Item=(Type=Bone,Name=\"spine_02\"),Weight=0.0403),(Item=(Type=Bone,Name=\"thigh_twist_01_l\"),Weight=0.0 ...(truncated)"
    Weight=1
    ExecutePin=@RigVMFunction_Sequence.B
   )
   > &RigUnit_ParentConstraint(
    AdvancedSettings=(RotationOrderForFilter=XZY)
    Child=(Type=Bone,Name="deform_skirt_m_l")
    Parents="((Item=(Type=Bone,Name=\"spine_03\"),Weight=0.01),(Item=(Type=Bone,Name=\"spine_01\"),Weight=0.0176),(Item=(Type=Bone,Name=\"spine_02\"),Weight=0.0208),(Item=(Type=Bone,Name=\"deform_groin_l\"),Weight=0.0562),(Item=(Type=Bone,Name=\"thigh_twist_01_l\"),Weight=0.0986 ...(truncated)"
    Weight=1
   )
   > &RigUnit_ParentConstraint(
    AdvancedSettings=(RotationOrderForFilter=XZY)
    Child=(Type=Bone,Name="deform_skirt_f_r")
    Parents="((Item=(Type=Bone,Name=\"thigh_twist_01_l\"),Weight=0.0008),(Item=(Type=Bone,Name=\"spine_03\"),Weight=0.0191),(Item=(Type=Bone,Name=\"spine_01\"),Weight=0.0292),(Item=(Type=Bone,Name=\"spine_02\"),Weight=0.0353),(Item=(Type=Bone,Name=\"thigh_twist_01_r\"),Weight=0. ...(truncated)"
    Weight=1
   )
   > &RigUnit_ParentConstraint(
    AdvancedSettings=(RotationOrderForFilter=XZY)
    Child=(Type=Bone,Name="deform_skirt_m_r")
    Parents="((Item=(Type=Bone,Name=\"spine_03\"),Weight=0.0095),(Item=(Type=Bone,Name=\"spine_01\"),Weight=0.0169),(Item=(Type=Bone,Name=\"spine_02\"),Weight=0.0199),(Item=(Type=Bone,Name=\"deform_groin_r\"),Weight=0.0559),(Item=(Type=Bone,Name=\"thigh_twist_01_r\"),Weight=0.10 ...(truncated)"
    Weight=1
   )
   > &RigUnit_ParentConstraint(
    AdvancedSettings=(RotationOrderForFilter=XZY)
    Child=(Type=Bone,Name="deform_skirt_b_l")
    Parents="((Item=(Type=Bone,Name=\"deform_groin_l\"),Weight=0.0009),(Item=(Type=Bone,Name=\"deform_glute_l\"),Weight=0.0185),(Item=(Type=Bone,Name=\"spine_03\"),Weight=0.0189),(Item=(Type=Bone,Name=\"thigh_twist_01_l\"),Weight=0.0244),(Item=(Type=Bone,Name=\"spine_01\"),Weigh ...(truncated)"
    Weight=1
   )
   > &RigUnit_ParentConstraint(
    AdvancedSettings=(RotationOrderForFilter=XZY)
    Child=(Type=Bone,Name="deform_skirt_bk_l")
    Parents=((Item=(Type=Bone,Name="deform_glute_l"),Weight=0.0068),(Item=(Type=Bone,Name="spine_03"),Weight=0.0203),(Item=(Type=Bone,Name="spine_01"),Weight=0.0302),(Item=(Type=Bone,Name="spine_02"),Weight=0.031),(Item=(Type=Bone,Name="pelvis"),Weight=0.9117))
    Weight=1
   )
   > &RigUnit_ParentConstraint(
    AdvancedSettings=(RotationOrderForFilter=XZY)
    Child=(Type=Bone,Name="deform_skirt_b_r")
    Parents="((Item=(Type=Bone,Name=\"deform_groin_r\"),Weight=0.001),(Item=(Type=Bone,Name=\"deform_glute_r\"),Weight=0.0158),(Item=(Type=Bone,Name=\"spine_03\"),Weight=0.0197),(Item=(Type=Bone,Name=\"thigh_twist_01_r\"),Weight=0.0244),(Item=(Type=Bone,Name=\"spine_01\"),Weight ...(truncated)"
    Weight=1
   )
   > &RigUnit_ParentConstraint(
    AdvancedSettings=(RotationOrderForFilter=XZY)
    Child=(Type=Bone,Name="deform_skirt_bk_r")
    Parents=((Item=(Type=Bone,Name="deform_glute_r"),Weight=0.0068),(Item=(Type=Bone,Name="spine_03"),Weight=0.0203),(Item=(Type=Bone,Name="spine_01"),Weight=0.0302),(Item=(Type=Bone,Name="spine_02"),Weight=0.031),(Item=(Type=Bone,Name="pelvis"),Weight=0.9117))
    Weight=1
   )
   > RigVMFunction_Sequence_1 = RigVMAggregateNode()
   n7 = &RigUnit_GetTransform(Item=(Type=Bone,Name="thigh_r"), bInitial=false)
   n8 = &RigVMFunction_MathQuaternionToEuler(RotationOrder=ZYX, Value=&RigUnit_GetTransform(Item=(Type=Control,Name="thigh_r_ctrl"), bInitial=False).Transform.Rotation)
   RigVMFunction_MathQuaternionFromEuler = &RigVMFunction_MathQuaternionFromEuler(
    Euler=(X=0,Y=0,Z=0)
    Euler.Y=@n8.Result.Y
    Euler.Z=@n8.Result.Z
    Euler.X=&RigVMFunction_MathQuaternionToEuler(RotationOrder=ZYX, Value=@n7.Transform.Rotation).Result.X
   )
   RigUnit_SetTransform = &RigUnit_SetTransform(
    Item=(Type=Control,Name="thigh_r_ctrl")
    Value=(Rotation=(X=0.2289,Y=0.6178,Z=-0.4226,W=0.6224),Translation=(X=-5.6207,Y=-4.5446,Z=84.2076),Scale3D=(X=1,Y=1,Z=1))
    Weight=1
    bInitial=False
    Value.Translation=@n7.Transform.Translation
    Value.Rotation=@RigVMFunction_MathQuaternionFromEuler.Result
    ExecutePin=@RigVMFunction_Sequence_1.A
   )
   > &RigUnit_AimConstraintLocalSpaceOffset(
    AdvancedSettings=(DebugSettings=(bEnabled=true,Scale=10,WorldOffset=(Rotation=(X=0,Y=0,Z=0,W=1),Translation=(X=0,Y=0,Z=0),Scale3D=(X=1,Y=1,Z=1))),RotationOrderForFilter=XZY)
    Child=(Type=Control,Name="thigh_r_ctrl")
    Filter=(bX=true,bY=true,bZ=true)
    Parents=((Item=(Type=Bone,Name="foot_r"),Weight=1))
    UpAxis=(X=0,Y=-1,Z=0)
    Weight=1
    WorldUp=(Target=(X=0,Y=0,Z=1),Kind=Location,Space=(Type=Bone,Name="thigh_r"))
    bMaintainOffset=false
   )
   > Get Delta Ref_3 = RigVMFunctionReferenceNode(Item=(Type=Control,Name="thigh_r_ctrl"))
   n9 = &RigVMFunction_AlphaInterp(
    InterpSpeedDecreasing=10
    InterpSpeedIncreasing=10
    Scale=1
    bClampResult=False
    Value=&RigVMFunction_MathBoolToFloat(Value=RigVMVariableNode() {variable=DBNO}.Value).Result
   )
   n10 = &RigVMFunction_ControlFlowBranch(
    Condition=RigVMDispatchNode(B=1, A=@n9.Result) {template="DISPATCH_RigVMDispatch_CoreEquals(in A,in B,out Result)"}.Result
    ExecuteContext=@"Get Delta Ref_3".ExecuteContext
   )
   n11 = &RigVMFunction_MathBoolToInteger(Value=/Script/RigVM.RigVMFunction_MathDoubleGreater(B=0, A=@"Get Delta Ref_3".Result.Rotation.Yaw) {is=struct RigVMUnitNode, template="Greater::Execute(in A,in B,out Result)"}.Result)
   n12 = &RigVMFunction_MathDoubleRemap(
    SourceMaximum=20
    SourceMinimum=0
    TargetMaximum=1
    TargetMinimum=0
    Value=&RigVMFunction_MathDoubleSub(
     A=@"Get Delta Ref_3".Result.Rotation.Yaw
     B=RigVMDispatchNode(Values=(-180,180), Index=@n11.Result) {template="DISPATCH_RigVMDispatch_SelectInt32(in Index,in Values,out Result)"}.Result
    ).Result
   )
   n13 = &RigVMFunction_MathDoubleRemap(SourceMaximum=-20, SourceMinimum=0, TargetMaximum=1, TargetMinimum=0, Value=@"Get Delta Ref_3".Result.Rotation.Pitch)
   n14 = &RigVMFunction_MathDoubleClamp(Value=&RigVMFunction_MathDoubleSub(A=@n12.Result, B=@n13.Result).Result)
   &RigUnit_RotationConstraintLocalSpaceOffset(
    AdvancedSettings=(InterpolationType=Average,RotationOrderForFilter=XYZ)
    Child=(Type=Bone,Name="deform_skirt_bk_r")
    Filter=(bX=false,bY=true,bZ=true)
    Parents=((Item=(Type=Control,Name="thigh_r_ctrl"),Weight=1))
    bMaintainOffset=false
    Weight=@n14.Result
    ExecutePin=@n10.False
   )
   n15 = &RigVMFunction_MathDoubleRemap(SourceMaximum=20, SourceMinimum=0, TargetMaximum=1, TargetMinimum=0, Value=@"Get Delta Ref_3".Result.Rotation.Pitch)
   > &RigUnit_RotationConstraintLocalSpaceOffset(
    AdvancedSettings=(InterpolationType=Average,RotationOrderForFilter=XYZ)
    Child=(Type=Bone,Name="deform_skirt_bk_l")
    Filter=(bX=false,bY=true,bZ=true)
    Parents=((Item=(Type=Control,Name="thigh_r_ctrl"),Weight=1))
    bMaintainOffset=false
    Weight=&RigVMFunction_MathDoubleClamp(Value=&RigVMFunction_MathDoubleSub(A=@n12.Result, B=@n15.Result).Result).Result
   )
   > &RigUnit_RotationConstraintLocalSpaceOffset(
    AdvancedSettings=(InterpolationType=Average,RotationOrderForFilter=XYZ)
    Child=(Type=Bone,Name="deform_skirt_b_l")
    Filter=(bX=false,bY=true,bZ=true)
    Parents=((Item=(Type=Control,Name="thigh_r_ctrl"),Weight=1))
    bMaintainOffset=false
    Weight=&RigVMFunction_MathDoubleMul(A=@n12.Result, B=@n13.Result).Result
   )
   > &RigUnit_RotationConstraintLocalSpaceOffset(
    AdvancedSettings=(InterpolationType=Average,RotationOrderForFilter=XYZ)
    Child=(Type=Bone,Name="deform_skirt_b_r")
    Filter=(bX=false,bY=true,bZ=true)
    Parents=((Item=(Type=Control,Name="thigh_r_ctrl"),Weight=1))
    bMaintainOffset=false
    Weight=@n14.Result
   )
   > Get Delta Ref_5 = RigVMFunctionReferenceNode(Item=(Type=Bone,Name="thigh_r"))
   > Set Transform_8 = &RigUnit_SetTranslation(
    Item=(Type=Bone,Name="deform_skirt_m_r")
    Space=GlobalSpace
    Weight=1
    bInitial=false
    bPropagateToChildren=true
    Value=&RigVMFunction_MathTransformClampSpatially(
     Axis=X
     DebugColor=(R=1,G=0,B=0,A=1)
     DebugThickness=1
     Maximum=70
     Minimum=9
     Type=Capsule
     bDrawDebug=false
     Value=&RigUnit_GetTransform(Item=(Type=Bone,Name="deform_skirt_m_r"), bInitial=false).Transform
     Space=&RigUnit_GetTransform(Item=(Type=Bone,Name="thigh_twist_01_r"), bInitial=false).Transform
    ) {template="ClampSpatially::Execute(in Value,in Axis,in Type,in Minimum,in Maximum,in Space,in bDrawDebug,in DebugColor,in DebugThickness,out Result)"}.Result.Translation
   ) {template="Set Transform::Execute(in Item,in Space,in bInitial,in Value,in Weight,in bPropagateToChildren)"}
   > &RigUnit_AimConstraintLocalSpaceOffset(
    Child=(Type=Bone,Name="deform_skirt_f_r")
    Filter=(bX=true,bY=true,bZ=true)
    Parents=((Item=(Type=Bone,Name="deform_knee_in_base_r"),Weight=1))
    UpAxis=(X=0,Y=-1,Z=0)
    Weight=1
    WorldUp=(Target=(X=0,Y=0,Z=1),Kind=Location,Space=(Type=Bone,Name="deform_skirt_f_r"))
   )
   > &RigUnit_AimConstraintLocalSpaceOffset(
    Child=(Type=Bone,Name="dyn_skirt_f_1_r")
    Filter=(bX=true,bY=true,bZ=true)
    Parents=((Item=(Type=Bone,Name="deform_knee_in_base_r"),Weight=1))
    UpAxis=(X=0,Y=-1,Z=0)
    Weight=1
    WorldUp=(Target=(X=0,Y=0,Z=1),Kind=Location,Space=(Type=Bone,Name="dyn_skirt_f_1_r"))
    bMaintainOffset=true
   )
   > &RigUnit_ModifyTransforms(
    ItemToModify="((Item=(Type=Bone,Name=\"deform_skirt_f_r\"),Transform=(Rotation=(X=0.1677,Y=-0.2549,Z=0.0449,W=0.9513),Translation=(X=0,Y=0,Z=2),Scale3D=(X=1,Y=1,Z=1))),(Item=(Type=Bone,Name=\"deform_skirt_m_r\"),Transform=(Rotation=(X=0,Y=0,Z=-0.2588,W=0.9659),Translation=( ...(truncated)"
    WeightMinimum=0
    Weight=&RigVMFunction_MathDoubleRemap(SourceMaximum=-90, SourceMinimum=0, TargetMaximum=1, TargetMinimum=0, Value=@"Get Delta Ref_5".Result.Rotation.Yaw).Result
   )
   > Set Transform_5 = &RigUnit_SetTranslation(
    Item=(Type=Bone,Name="deform_skirt_f_r")
    Space=GlobalSpace
    Weight=1
    bInitial=false
    bPropagateToChildren=true
    Value=&RigVMFunction_MathTransformClampSpatially(
     Axis=X
     DebugColor=(R=1,G=0,B=0,A=1)
     DebugThickness=1
     Maximum=70
     Minimum=9
     Type=Capsule
     bDrawDebug=false
     Value=&RigUnit_GetTransform(Item=(Type=Bone,Name="deform_skirt_f_r"), bInitial=false).Transform
     Space=&RigUnit_GetTransform(Item=(Type=Bone,Name="thigh_twist_01_r"), bInitial=false).Transform
    ) {template="ClampSpatially::Execute(in Value,in Axis,in Type,in Minimum,in Maximum,in Space,in bDrawDebug,in DebugColor,in DebugThickness,out Result)"}.Result.Translation
   ) {template="Set Transform::Execute(in Item,in Space,in bInitial,in Value,in Weight,in bPropagateToChildren)"}
   > Set Transform_7 = &RigUnit_SetTranslation(
    Item=(Type=Bone,Name="dyn_skirt_f_1_r")
    Space=GlobalSpace
    Weight=1
    bInitial=false
    bPropagateToChildren=true
    Value=&RigVMFunction_MathTransformClampSpatially(
     Axis=X
     DebugColor=(R=1,G=0,B=0,A=1)
     DebugThickness=1
     Maximum=70
     Minimum=11
     Type=Capsule
     bDrawDebug=false
     Value=&RigUnit_GetTransform(Item=(Type=Bone,Name="dyn_skirt_f_1_r"), bInitial=false).Transform
     Space=&RigUnit_GetTransform(Item=(Type=Bone,Name="thigh_twist_01_r"), bInitial=false).Transform
    ) {template="ClampSpatially::Execute(in Value,in Axis,in Type,in Minimum,in Maximum,in Space,in bDrawDebug,in DebugColor,in DebugThickness,out Result)"}.Result.Translation
   ) {template="Set Transform::Execute(in Item,in Space,in bInitial,in Value,in Weight,in bPropagateToChildren)"}
   > &RigUnit_RotationConstraintLocalSpaceOffset(
    AdvancedSettings=(InterpolationType=Average,RotationOrderForFilter=XZY)
    Child=(Type=Bone,Name="deform_skirt_m_r")
    Filter=(bX=true,bY=true,bZ=true)
    Parents=((Item=(Type=Bone,Name="deform_skirt_f_r"),Weight=0.8),(Item=(Type=Bone,Name="deform_skirt_b_r"),Weight=0.2))
    bMaintainOffset=false
    Weight=@n15.Result
   )
   Get Delta Ref_2 = RigVMFunctionReferenceNode(Item=(Type=Bone,Name="thigh_r"), ExecuteContext=@n10.True)
   > Set Transform_3 = &RigUnit_SetTranslation(
    Item=(Type=Bone,Name="deform_skirt_m_r")
    Space=GlobalSpace
    Weight=1
    bInitial=false
    bPropagateToChildren=true
    Value=&RigVMFunction_MathTransformClampSpatially(
     Axis=X
     DebugColor=(R=1,G=0,B=0,A=1)
     DebugThickness=1
     Maximum=70
     Minimum=10
     Type=Capsule
     bDrawDebug=false
     Value=&RigUnit_GetTransform(Item=(Type=Bone,Name="deform_skirt_m_r"), bInitial=false).Transform
     Space=&RigUnit_GetTransform(Item=(Type=Bone,Name="thigh_twist_01_r"), bInitial=false).Transform
    ) {template="ClampSpatially::Execute(in Value,in Axis,in Type,in Minimum,in Maximum,in Space,in bDrawDebug,in DebugColor,in DebugThickness,out Result)"}.Result.Translation
   ) {template="Set Transform::Execute(in Item,in Space,in bInitial,in Value,in Weight,in bPropagateToChildren)"}
   > &RigUnit_AimConstraintLocalSpaceOffset(
    Child=(Type=Bone,Name="deform_skirt_f_r")
    Filter=(bX=true,bY=true,bZ=true)
    Parents=((Item=(Type=Bone,Name="calf_r"),Weight=1))
    UpAxis=(X=0,Y=-1,Z=0)
    Weight=1
    WorldUp=(Target=(X=0,Y=0,Z=1),Kind=Location,Space=(Type=Bone,Name="deform_skirt_f_r"))
   )
   > &RigUnit_ModifyTransforms(
    ItemToModify="((Item=(Type=Bone,Name=\"deform_skirt_f_r\"),Transform=(Rotation=(X=0,Y=-0.2588,Z=0,W=0.9659),Translation=(X=0,Y=0,Z=0),Scale3D=(X=1,Y=1,Z=1))),(Item=(Type=Bone,Name=\"deform_skirt_m_r\"),Transform=(Rotation=(X=0,Y=0,Z=-0.2588,W=0.9659),Translation=(X=0,Y=0,Z= ...(truncated)"
    WeightMinimum=0
    Weight=&RigVMFunction_MathDoubleRemap(SourceMaximum=-90, SourceMinimum=0, TargetMaximum=1, TargetMinimum=0, Value=@"Get Delta Ref_2".Result.Rotation.Yaw).Result
   )
   n16 = &RigUnit_GetTransform(Item=(Type=Bone,Name="thigh_l"), bInitial=false)
   n17 = &RigVMFunction_MathQuaternionToEuler(RotationOrder=ZYX, Value=&RigUnit_GetTransform(Item=(Type=Control,Name="thigh_l_ctrl"), bInitial=False).Transform.Rotation)
   &RigUnit_SetTransform(
    Item=(Type=Control,Name="thigh_l_ctrl")
    Value=(Rotation=(X=0.2289,Y=0.6178,Z=-0.4226,W=0.6224),Translation=(X=-5.6207,Y=-4.5446,Z=84.2076),Scale3D=(X=1,Y=1,Z=1))
    Weight=1
    bInitial=False
    Value.Translation=@n16.Transform.Translation
    Value.Rotation=&RigVMFunction_MathQuaternionFromEuler(
     Euler=(X=0,Y=0,Z=0)
     Euler.Y=@n17.Result.Y
     Euler.Z=@n17.Result.Z
     Euler.X=&RigVMFunction_MathQuaternionToEuler(RotationOrder=ZYX, Value=@n16.Transform.Rotation).Result.X
    ).Result
    ExecutePin=@RigVMFunction_Sequence_1.B
   )
   > &RigUnit_AimConstraintLocalSpaceOffset(
    AdvancedSettings=(DebugSettings=(bEnabled=true,Scale=10,WorldOffset=(Rotation=(X=0,Y=0,Z=0,W=1),Translation=(X=0,Y=0,Z=0),Scale3D=(X=1,Y=1,Z=1))),RotationOrderForFilter=XYZ)
    Child=(Type=Control,Name="thigh_l_ctrl")
    Filter=(bX=true,bY=true,bZ=true)
    Parents=((Item=(Type=Bone,Name="foot_l"),Weight=1))
    UpAxis=(X=0,Y=-1,Z=0)
    Weight=1
    WorldUp=(Target=(X=0,Y=0,Z=1),Kind=Location,Space=(Type=Bone,Name="thigh_l"))
    bMaintainOffset=false
   )
   > Get Delta Ref_1 = RigVMFunctionReferenceNode(Item=(Type=Control,Name="thigh_l_ctrl"))
   n18 = &RigVMFunction_AlphaInterp(
    InterpSpeedDecreasing=10
    InterpSpeedIncreasing=10
    Scale=1
    bClampResult=False
    Value=&RigVMFunction_MathBoolToFloat(Value=RigVMVariableNode() {variable=DBNO}.Value).Result
   )
   n19 = &RigVMFunction_ControlFlowBranch(
    Condition=RigVMDispatchNode(B=1, A=@n18.Result) {template="DISPATCH_RigVMDispatch_CoreEquals(in A,in B,out Result)"}.Result
    ExecuteContext=@"Get Delta Ref_1".ExecuteContext
   )
   n20 = &RigVMFunction_MathDoubleRemap(SourceMaximum=20, SourceMinimum=0, TargetMaximum=1, TargetMinimum=0, Value=@"Get Delta Ref_1".Result.Rotation.Pitch)
   Greater = &RigVMFunction_MathDoubleGreater(B=0, A=@"Get Delta Ref_1".Result.Rotation.Yaw) {template="Greater::Execute(in A,in B,out Result)"}
   RigVMFunction_MathBoolToInteger = &RigVMFunction_MathBoolToInteger(Value=@Greater.Result)
   n21 = &RigVMFunction_MathDoubleRemap(
    SourceMaximum=20
    SourceMinimum=0
    TargetMaximum=1
    TargetMinimum=0
    Value=&RigVMFunction_MathDoubleSub(
     A=@"Get Delta Ref_1".Result.Rotation.Yaw
     B=RigVMDispatchNode(Values=(-180,180), Index=@RigVMFunction_MathBoolToInteger.Result) {template="DISPATCH_RigVMDispatch_SelectInt32(in Index,in Values,out Result)"}.Result
    ).Result
   )
   n22 = &RigVMFunction_MathDoubleClamp(Value=&RigVMFunction_MathDoubleSub(B=@n20.Result, A=@n21.Result).Result)
   &RigUnit_RotationConstraintLocalSpaceOffset(
    AdvancedSettings=(InterpolationType=Average,RotationOrderForFilter=XYZ)
    Child=(Type=Bone,Name="deform_skirt_bk_l")
    Filter=(bX=false,bY=true,bZ=true)
    Parents=((Item=(Type=Control,Name="thigh_l_ctrl"),Weight=1))
    bMaintainOffset=false
    Weight=@n22.Result
    ExecutePin=@n19.False
   )
   n23 = &RigVMFunction_MathDoubleRemap(SourceMaximum=-20, SourceMinimum=0, TargetMaximum=1, TargetMinimum=0, Value=@"Get Delta Ref_1".Result.Rotation.Pitch)
   Clamp = &RigVMFunction_MathDoubleClamp(Value=&RigVMFunction_MathDoubleSub(B=@n23.Result, A=@n21.Result).Result)
   > &RigUnit_RotationConstraintLocalSpaceOffset(
    AdvancedSettings=(InterpolationType=Average,RotationOrderForFilter=XYZ)
    Child=(Type=Bone,Name="deform_skirt_bk_r")
    Filter=(bX=false,bY=true,bZ=true)
    Parents=((Item=(Type=Control,Name="thigh_l_ctrl"),Weight=1))
    bMaintainOffset=false
    Weight=@Clamp.Result
   )
   Multiply = &RigVMFunction_MathDoubleMul(B=@n20.Result, A=@n21.Result)
   > &RigUnit_RotationConstraintLocalSpaceOffset(
    AdvancedSettings=(InterpolationType=Average,RotationOrderForFilter=XYZ)
    Child=(Type=Bone,Name="deform_skirt_b_r")
    Filter=(bX=false,bY=true,bZ=true)
    Parents=((Item=(Type=Control,Name="thigh_l_ctrl"),Weight=1))
    bMaintainOffset=false
    Weight=@Multiply.Result
   )
   > &RigUnit_RotationConstraintLocalSpaceOffset(
    AdvancedSettings=(InterpolationType=Average,RotationOrderForFilter=XYZ)
    Child=(Type=Bone,Name="deform_skirt_b_l")
    Filter=(bX=false,bY=true,bZ=true)
    Parents=((Item=(Type=Control,Name="thigh_l_ctrl"),Weight=1))
    bMaintainOffset=false
    Weight=@n22.Result
   )
   > Get Delta Ref_4 = RigVMFunctionReferenceNode(Item=(Type=Bone,Name="thigh_l"))
   > Set Transform_1 = &RigUnit_SetTranslation(
    Item=(Type=Bone,Name="deform_skirt_m_l")
    Space=GlobalSpace
    Weight=1
    bInitial=false
    bPropagateToChildren=true
    Value=&RigVMFunction_MathTransformClampSpatially(
     Axis=X
     DebugColor=(R=1,G=0,B=0,A=1)
     DebugThickness=1
     Maximum=70
     Minimum=10
     Type=Capsule
     bDrawDebug=false
     Space=&RigUnit_GetTransform(Item=(Type=Bone,Name="thigh_twist_01_l"), bInitial=false).Transform
     Value=&RigUnit_GetTransform(Item=(Type=Bone,Name="deform_skirt_m_l"), bInitial=false).Transform
    ) {template="ClampSpatially::Execute(in Value,in Axis,in Type,in Minimum,in Maximum,in Space,in bDrawDebug,in DebugColor,in DebugThickness,out Result)"}.Result.Translation
   ) {template="Set Transform::Execute(in Item,in Space,in bInitial,in Value,in Weight,in bPropagateToChildren)"}
   > &RigUnit_AimConstraintLocalSpaceOffset(
    Child=(Type=Bone,Name="deform_skirt_f_l")
    Filter=(bX=true,bY=true,bZ=true)
    Parents=((Item=(Type=Bone,Name="calf_l"),Weight=1))
    UpAxis=(X=0,Y=-1,Z=0)
    Weight=1
    WorldUp=(Target=(X=0,Y=0,Z=1),Kind=Location,Space=(Type=Bone,Name="deform_skirt_f_l"))
    bMaintainOffset=true
   )
   > &RigUnit_AimConstraintLocalSpaceOffset(
    Child=(Type=Bone,Name="dyn_skirt_f_1_l")
    Filter=(bX=true,bY=true,bZ=true)
    Parents=((Item=(Type=Bone,Name="thigh_twist_01_l"),Weight=1))
    UpAxis=(X=0,Y=-1,Z=0)
    Weight=1
    WorldUp=(Target=(X=0,Y=0,Z=1),Kind=Location,Space=(Type=Bone,Name="dyn_skirt_f_1_l"))
    bMaintainOffset=true
   )
   n24 = &RigVMFunction_MathDoubleRemap(SourceMaximum=-90, SourceMinimum=0, TargetMaximum=1, TargetMinimum=0, Value=@"Get Delta Ref_4".Result.Rotation.Yaw)
   > &RigUnit_ModifyTransforms(
    ItemToModify="((Item=(Type=Bone,Name=\"deform_skirt_f_l\"),Transform=(Rotation=(X=-0.1632,Y=0.3368,Z=0.0594,W=0.9254),Translation=(X=0,Y=0,Z=0),Scale3D=(X=1,Y=1,Z=1))),(Item=(Type=Bone,Name=\"deform_skirt_m_l\"),Transform=(Rotation=(X=0,Y=0,Z=-0.3007,W=0.9537),Translation=( ...(truncated)"
    WeightMinimum=0
    Weight=@n24.Result
   )
   RigVMFunction_AnimEvalRichCurve = &RigVMFunction_AnimEvalRichCurve(Curve=(EditorCurveData=(Keys=((),(Time=0.3246,Value=0.0499),(Time=1,Value=1)))), SourceMinimum=0, TargetMinimum=0, Value=@n24.Result)
   > &RigUnit_ModifyTransforms(
    ItemToModify=((Item=(Type=Bone,Name="dyn_skirt_f_1_l"),Transform=(Rotation=(X=0.1677,Y=-0.2549,Z=0.0449,W=0.9513),Translation=(X=0,Y=0,Z=0),Scale3D=(X=1,Y=1,Z=1))))
    WeightMinimum=0
    Weight=@RigVMFunction_AnimEvalRichCurve.Result
   )
   > Set Transform_9 = &RigUnit_SetTranslation(
    Item=(Type=Bone,Name="deform_skirt_f_l")
    Space=GlobalSpace
    Weight=1
    bInitial=false
    bPropagateToChildren=true
    Value=&RigVMFunction_MathTransformClampSpatially(
     Axis=X
     DebugColor=(R=1,G=0,B=0,A=1)
     DebugThickness=1
     Maximum=70
     Minimum=8
     Type=Capsule
     bDrawDebug=false
     Value=&RigUnit_GetTransform(Item=(Type=Bone,Name="deform_skirt_f_l"), bInitial=false).Transform
     Space=&RigUnit_GetTransform(Item=(Type=Bone,Name="thigh_twist_01_l"), bInitial=false).Transform
    ) {template="ClampSpatially::Execute(in Value,in Axis,in Type,in Minimum,in Maximum,in Space,in bDrawDebug,in DebugColor,in DebugThickness,out Result)"}.Result.Translation
   ) {template="Set Transform::Execute(in Item,in Space,in bInitial,in Value,in Weight,in bPropagateToChildren)"}
   > Set Transform_6 = &RigUnit_SetTranslation(
    Item=(Type=Bone,Name="dyn_skirt_f_1_l")
    Space=GlobalSpace
    Weight=1
    bInitial=false
    bPropagateToChildren=true
    Value=&RigVMFunction_MathTransformClampSpatially(
     Axis=X
     DebugColor=(R=1,G=0,B=0,A=1)
     DebugThickness=1
     Maximum=70
     Minimum=10
     Type=Capsule
     bDrawDebug=false
     Value=&RigUnit_GetTransform(Item=(Type=Bone,Name="dyn_skirt_f_1_l"), bInitial=false).Transform
     Space=&RigUnit_GetTransform(Item=(Type=Bone,Name="thigh_twist_01_l"), bInitial=false).Transform
    ) {template="ClampSpatially::Execute(in Value,in Axis,in Type,in Minimum,in Maximum,in Space,in bDrawDebug,in DebugColor,in DebugThickness,out Result)"}.Result.Translation
   ) {template="Set Transform::Execute(in Item,in Space,in bInitial,in Value,in Weight,in bPropagateToChildren)"}
   > &RigUnit_RotationConstraintLocalSpaceOffset(
    AdvancedSettings=(InterpolationType=Average,RotationOrderForFilter=XZY)
    Child=(Type=Bone,Name="deform_skirt_m_l")
    Filter=(bX=true,bY=true,bZ=true)
    Parents=((Item=(Type=Bone,Name="deform_skirt_f_l"),Weight=0.5),(Item=(Type=Bone,Name="deform_skirt_b_l"),Weight=0.5))
    bMaintainOffset=false
    Weight=@n23.Result
   )
   Get Delta Ref_6 = RigVMFunctionReferenceNode(Item=(Type=Bone,Name="thigh_l"), ExecuteContext=@n19.True)
   > Set Transform_4 = &RigUnit_SetTranslation(
    Item=(Type=Bone,Name="deform_skirt_m_l")
    Space=GlobalSpace
    Weight=1
    bInitial=false
    bPropagateToChildren=true
    Value=&RigVMFunction_MathTransformClampSpatially(
     Axis=X
     DebugColor=(R=1,G=0,B=0,A=1)
     DebugThickness=1
     Maximum=70
     Minimum=9
     Type=Capsule
     bDrawDebug=false
     Value=&RigUnit_GetTransform(Item=(Type=Bone,Name="deform_skirt_m_l"), bInitial=false).Transform
     Space=&RigUnit_GetTransform(Item=(Type=Bone,Name="thigh_twist_01_l"), bInitial=false).Transform
    ) {template="ClampSpatially::Execute(in Value,in Axis,in Type,in Minimum,in Maximum,in Space,in bDrawDebug,in DebugColor,in DebugThickness,out Result)"}.Result.Translation
   ) {template="Set Transform::Execute(in Item,in Space,in bInitial,in Value,in Weight,in bPropagateToChildren)"}
   > &RigUnit_AimConstraintLocalSpaceOffset(
    Child=(Type=Bone,Name="deform_skirt_f_l")
    Filter=(bX=true,bY=true,bZ=true)
    Parents=((Item=(Type=Bone,Name="calf_l"),Weight=1))
    UpAxis=(X=0,Y=-1,Z=0)
    Weight=1
    WorldUp=(Target=(X=0,Y=0,Z=1),Kind=Location,Space=(Type=Bone,Name="deform_skirt_f_l"))
    bMaintainOffset=true
   )
   > &RigUnit_ModifyTransforms(
    ItemToModify="((Item=(Type=Bone,Name=\"deform_skirt_f_l\"),Transform=(Rotation=(X=0,Y=0.5,Z=0,W=0.866),Translation=(X=0,Y=0,Z=0),Scale3D=(X=1,Y=1,Z=1))),(Item=(Type=Bone,Name=\"deform_skirt_m_l\"),Transform=(Rotation=(X=0,Y=0,Z=-0.2588,W=0.9659),Translation=(X=0,Y=0,Z=0),Sc ...(truncated)"
    WeightMinimum=0
    Weight=&RigVMFunction_MathDoubleRemap(SourceMaximum=-90, SourceMinimum=0, TargetMaximum=1, TargetMinimum=0, Value=@"Get Delta Ref_6".Result.Rotation.Yaw).Result
   )
  ]}
 RigVMFunctionLibrary {nodes [Belt and Skirt Setup = RigVMCollapseNode(ExecuteContext=())]}
 RigVMFunction_MathBoolOr_1_ContainedGraph {nodes [
   n1 = RigVMFunctionEntryNode()
   RigVMFunction_MathBoolOr = &RigVMFunction_MathBoolOr(A=@n1.A, B=@n1.B)
   RigVMFunctionReturnNode(Result=&RigVMFunction_MathBoolOr(B=@n1.C, A=@RigVMFunction_MathBoolOr.Result).Result)
  ]}
 RigVMFunction_Sequence_1_ContainedGraph {nodes [
   n1 = RigVMFunctionEntryNode()
   RigVMFunction_Sequence = &RigVMFunction_Sequence(ExecuteContext=@n1.ExecuteContext)
   n2 = &RigVMFunction_Sequence(ExecuteContext=@RigVMFunction_Sequence.B)
   n3 = &RigVMFunction_Sequence(ExecuteContext=@n2.B)
   n4 = &RigVMFunction_Sequence(ExecuteContext=@n3.B)
   RigVMFunctionReturnNode(A=@RigVMFunction_Sequence.A, B=@n2.A, C=@n3.A, D=@n4.A, E=@n4.B)
  ]}
 RigVMFunction_Sequence_2_ContainedGraph {nodes [
   n1 = RigVMFunctionEntryNode()
   RigVMFunction_Sequence = &RigVMFunction_Sequence(ExecuteContext=@n1.ExecuteContext)
   n2 = &RigVMFunction_Sequence(ExecuteContext=@RigVMFunction_Sequence.B)
   n3 = &RigVMFunction_Sequence(ExecuteContext=@n2.B)
   n4 = &RigVMFunction_Sequence(ExecuteContext=@n3.B)
   RigVMFunctionReturnNode(A=@RigVMFunction_Sequence.A, B=@n2.A, C=@n3.A, D=@n4.A, E=@n4.B)
  ]}
 RigVMFunction_Sequence_2_ContainedGraph {nodes [
   n1 = RigVMFunctionEntryNode()
   n2 = &RigVMFunction_Sequence(ExecuteContext=@n1.ExecuteContext)
   n3 = &RigVMFunction_Sequence(ExecuteContext=@n2.B)
   RigVMFunctionReturnNode(A=@n2.A, B=@n3.A, C=@n3.B)
  ]}
 RigVMModel {nodes [
   RigUnit_BeginExecution = &RigUnit_BeginExecution()
   FN_DeformRig_FNC_1 = RigVMFunctionReferenceNode(
    DeformRig_ Curve="(deformRig_l_upperarm_up_alpha_72_#1=False,deformRig_r_upperarm_up_alpha_73_#2=False,deformRig_l_upperarm_fr_alpha_74_#3=false,deformRig_r_upperarm_fr_alpha_75_CB51B8 ...(truncated)"
    DeformRig_Bool="(upperarm_1_#4=true,elbow_3_#5=true,wrist_7_#6=true,groin_8_#7=True,glute_17_#8=True,knee_9_40C408324CA261 ...(truncated)"
    DeformRig_Debug=(upperarm_debug_18_#4=False,wrist_debug_20_#6=False,lat_debug_26_#9=False,groinGlute_debug_27_#7=False)
    DeformRig_Value="(upperarm_up_18_#10=(Rotation=(X=0.0,Y=-0.0,Z=0.0,W=1.0),Translation=(X=0.0,Y=3,Z=8),Scale3D=(X=0.70,Y=1.0,Z=1.0)),upperarm_fwd_fr_19_#11=(Rotation=(X=0.0,Y=-0.0,Z=0.0,W=1.0),Translation=(X=-10,Y=10 ...(truncated)"
    ExecuteContext=@RigUnit_BeginExecution.ExecutePin
   )
   > RigVMFunction_Sequence = RigVMAggregateNode()
   &RigUnit_ParentConstraint(
    AdvancedSettings=(RotationOrderForFilter=XZY)
    Child=(Type=Bone,Name="deform_collar_f_l")
    Parents="((Item=(Type=Bone,Name=\"deform_pec_l\"),Weight=0.0037),(Item=(Type=Bone,Name=\"spine_03\"),Weight=0.0041),(Item=(Type=Bone,Name=\"spine_04\"),Weight=0.0162),(Item=(Type=Bone,Name=\"head\"),Weight=0.028),(Item=(Type=Bone,Name=\"neck_01\"),Weight=0.0557),(Item=(Type= ...(truncated)"
    Weight=1
    ExecutePin=@RigVMFunction_Sequence.A
   )
   > &RigUnit_ParentConstraint(
    AdvancedSettings=(RotationOrderForFilter=XZY)
    Child=(Type=Bone,Name="deform_collar_l")
    Parents="((Item=(Type=Bone,Name=\"spine_03\"),Weight=0.0007),(Item=(Type=Bone,Name=\"deform_pec_l\"),Weight=0.0013),(Item=(Type=Bone,Name=\"spine_04\"),Weight=0.0023),(Item=(Type=Bone,Name=\"head\"),Weight=0.0219),(Item=(Type=Bone,Name=\"neck_01\"),Weight=0.0556),(Item=(Type ...(truncated)"
    Weight=1
   )
   > &RigUnit_ParentConstraint(
    AdvancedSettings=(RotationOrderForFilter=XZY)
    Child=(Type=Bone,Name="deform_collar_b_l")
    Parents=((Item=(Type=Bone,Name="clavicle_l"),Weight=0.0308),(Item=(Type=Bone,Name="head"),Weight=0.0477),(Item=(Type=Bone,Name="neck_01"),Weight=0.0967),(Item=(Type=Bone,Name="neck_02"),Weight=0.2322),(Item=(Type=Bone,Name="spine_05"),Weight=0.5926))
    Weight=1
   )
   > &RigUnit_ParentConstraint(
    AdvancedSettings=(RotationOrderForFilter=XZY)
    Child=(Type=Bone,Name="deform_collar_bk")
    Parents=((Item=(Type=Bone,Name="head"),Weight=0.1023),(Item=(Type=Bone,Name="neck_01"),Weight=0.1247),(Item=(Type=Bone,Name="neck_02"),Weight=0.2913),(Item=(Type=Bone,Name="spine_05"),Weight=0.4817))
    Weight=1
   )
   > &RigUnit_ParentConstraint(
    AdvancedSettings=(RotationOrderForFilter=XZY)
    Child=(Type=Bone,Name="deform_collar_f_r")
    Parents="((Item=(Type=Bone,Name=\"deform_pec_r\"),Weight=0.0037),(Item=(Type=Bone,Name=\"spine_03\"),Weight=0.0041),(Item=(Type=Bone,Name=\"spine_04\"),Weight=0.0161),(Item=(Type=Bone,Name=\"head\"),Weight=0.028),(Item=(Type=Bone,Name=\"neck_01\"),Weight=0.0557),(Item=(Type= ...(truncated)"
    Weight=1
   )
   > &RigUnit_ParentConstraint(
    AdvancedSettings=(RotationOrderForFilter=XZY)
    Child=(Type=Bone,Name="deform_collar_r")
    Parents="((Item=(Type=Bone,Name=\"deform_pec_r\"),Weight=0.0013),(Item=(Type=Bone,Name=\"spine_04\"),Weight=0.0016),(Item=(Type=Bone,Name=\"head\"),Weight=0.0077),(Item=(Type=Bone,Name=\"neck_01\"),Weight=0.0188),(Item=(Type=Bone,Name=\"clavicle_r\"),Weight=0.0392),(Item=(Ty ...(truncated)"
    Weight=1
   )
   > &RigUnit_ParentConstraint(
    AdvancedSettings=(RotationOrderForFilter=XZY)
    Child=(Type=Bone,Name="deform_collar_b_r")
    Parents=((Item=(Type=Bone,Name="clavicle_r"),Weight=0.0313),(Item=(Type=Bone,Name="head"),Weight=0.048),(Item=(Type=Bone,Name="neck_01"),Weight=0.0964),(Item=(Type=Bone,Name="neck_02"),Weight=0.232),(Item=(Type=Bone,Name="spine_05"),Weight=0.5923))
    Weight=1
   )
   > &RigUnit_ParentConstraint(
    AdvancedSettings=(RotationOrderForFilter=XZY)
    Child=(Type=Bone,Name="deform_chain_up_inn_l")
    Parents=((Item=(Type=Bone,Name="deform_pec_aim_l"),Weight=0.0001),(Item=(Type=Bone,Name="clavicle_l"),Weight=0.0004),(Item=(Type=Bone,Name="neck_01"),Weight=0.0162),(Item=(Type=Bone,Name="spine_05"),Weight=0.9833))
    Weight=1
   )
   > &RigUnit_ParentConstraint(
    AdvancedSettings=(RotationOrderForFilter=XZY)
    Child=(Type=Bone,Name="deform_chain_up_out_l")
    Parents="((Item=(Type=Bone,Name=\"upperarm_twist_02_l\"),Weight=0.03),(Item=(Type=Bone,Name=\"spine_04\"),Weight=0.0379),(Item=(Type=Bone,Name=\"deform_upperarm_up_l\"),Weight=0.0558),(Item=(Type=Bone,Name=\"deform_upperarm_fr_l\"),Weight=0.1285),(Item=(Type=Bone,Name=\"spi ...(truncated)"
    Weight=1
   )
   > &RigUnit_ParentConstraint(
    AdvancedSettings=(RotationOrderForFilter=XZY)
    Child=(Type=Bone,Name="deform_chain_down_inn_l")
    Parents=((Item=(Type=Bone,Name="neck_01"),Weight=0.0022),(Item=(Type=Bone,Name="spine_05"),Weight=0.9978))
    Weight=1
   )
   > &RigUnit_ParentConstraint(
    AdvancedSettings=(RotationOrderForFilter=XZY)
    Child=(Type=Bone,Name="deform_chain_up_inn_r")
    Parents="((Item=(Type=Bone,Name=\"deform_pec_aim_r\"),Weight=0.0027),(Item=(Type=Bone,Name=\"spine_04\"),Weight=0.0064),(Item=(Type=Bone,Name=\"neck_02\"),Weight=0.0171),(Item=(Type=Bone,Name=\"dyn_collar_r\"),Weight=0.0254),(Item=(Type=Bone,Name=\"clavicle_r\"),Weight=0.028 ...(truncated)"
    Weight=1
   )
   > &RigUnit_ParentConstraint(
    AdvancedSettings=(RotationOrderForFilter=XZY)
    Child=(Type=Bone,Name="deform_chain_up_out_r")
    Parents="((Item=(Type=Bone,Name=\"upperarm_twist_02_r\"),Weight=0.0276),(Item=(Type=Bone,Name=\"spine_04\"),Weight=0.0349),(Item=(Type=Bone,Name=\"deform_upperarm_up_r\"),Weight=0.0514),(Item=(Type=Bone,Name=\"deform_pec_r\"),Weight=0.0791),(Item=(Type=Bone,Name=\"deform_up ...(truncated)"
    Weight=1
   )
   > &RigUnit_ParentConstraint(
    AdvancedSettings=(RotationOrderForFilter=XZY)
    Child=(Type=Bone,Name="deform_chain_down_inn_r")
    Parents=((Item=(Type=Bone,Name="neck_01"),Weight=0.0022),(Item=(Type=Bone,Name="spine_05"),Weight=0.9978))
    Weight=1
   )
   n1 = &RigUnit_SphericalPoseReader(
    ActiveRegionScaleFactors=(PositiveWidth=2.5,NegativeWidth=3,PositiveHeight=3,NegativeHeight=3)
    ActiveRegionSize=0.1
    Debug=(bDrawDebug=false,DebugScale=25,DebugSegments=20,DebugThickness=0.25)
    DriverAxis=(X=1,Y=0,Z=0)
    DriverItem=(Type=Bone,Name="upperarm_r")
    FalloffRegionScaleFactors=(PositiveWidth=1,NegativeWidth=1,PositiveHeight=1,NegativeHeight=1)
    FalloffSize=0.7
    OptionalParentItem=(Type=Bone)
    RotationOffset=(X=0,Y=0,Z=0)
    ExecutePin=@RigVMFunction_Sequence.B
   )
   > &RigUnit_ModifyTransforms(
    ItemToModify="((Item=(Type=Bone,Name=\"deform_upperarm_bk_r\"),Transform=(Rotation=(X=0,Y=0,Z=0,W=1),Translation=(X=-2,Y=-3,Z=15),Scale3D=(X=1,Y=1,Z=1))),(Item=(Type=Bone,Name=\"upperarm_twist_02_r\"),Transform=(Rotation=(X=0,Y=0,Z=0,W=1),Translation=(X=0,Y=0,Z=0),Scale3D=( ...(truncated)"
    Weight=@n1.OutputParam
   )
   > n2 = &RigUnit_SphericalPoseReader(
    ActiveRegionScaleFactors=(PositiveWidth=2.5,NegativeWidth=3,PositiveHeight=3,NegativeHeight=3)
    ActiveRegionSize=0.1
    Debug=(bDrawDebug=false,DebugScale=25,DebugSegments=20,DebugThickness=0.25)
    DriverAxis=(X=1,Y=0,Z=0)
    DriverItem=(Type=Bone,Name="upperarm_l")
    FalloffRegionScaleFactors=(PositiveWidth=1,NegativeWidth=1,PositiveHeight=1,NegativeHeight=1)
    FalloffSize=0.7
    OptionalParentItem=(Type=Bone)
    RotationOffset=(X=0,Y=0,Z=0)
   )
   > &RigUnit_ModifyTransforms(
    ItemToModify="((Item=(Type=Bone,Name=\"deform_upperarm_bk_l\"),Transform=(Rotation=(X=0,Y=0,Z=0,W=1),Translation=(X=2,Y=3,Z=-12),Scale3D=(X=1,Y=1,Z=1))),(Item=(Type=Bone,Name=\"upperarm_twist_02_l\"),Transform=(Rotation=(X=0,Y=0,Z=0,W=1),Translation=(X=0,Y=0,Z=0),Scale3D=(X ...(truncated)"
    Weight=@n2.OutputParam
   )
   Belt and Skirt Setup = RigVMFunctionReferenceNode(ExecuteContext=@RigVMFunction_Sequence.C)
   &RigUnit_AimConstraintLocalSpaceOffset(
    Child=(Type=Bone,Name="deform_collar_bk")
    Filter=(bX=false,bY=false,bZ=True)
    Parents=((Item=(Type=Control,Name="head_ctrl"),Weight=1))
    UpAxis=(X=0,Y=-1,Z=0)
    Weight=1
    WorldUp=(Target=(X=0,Y=0,Z=1),Kind=Location,Space=(Type=Bone,Name="deform_collar_bk"))
    ExecutePin=@RigVMFunction_Sequence.D
   )
   > &RigUnit_AimConstraintLocalSpaceOffset(
    Child=(Type=Bone,Name="deform_collar_b_r")
    Filter=(bX=true,bY=true,bZ=True)
    Parents=((Item=(Type=Bone,Name="head"),Weight=1))
    UpAxis=(X=0,Y=-1,Z=0)
    Weight=1
    WorldUp=(Target=(X=0,Y=0,Z=1),Kind=Location,Space=(Type=Bone,Name="deform_collar_b_r"))
   )
   > &RigUnit_AimConstraintLocalSpaceOffset(
    Child=(Type=Bone,Name="deform_collar_b_l")
    Filter=(bX=True,bY=true,bZ=True)
    Parents=((Item=(Type=Bone,Name="head"),Weight=1))
    UpAxis=(X=0,Y=-1,Z=0)
    Weight=1
    WorldUp=(Target=(X=0,Y=0,Z=1),Kind=Location,Space=(Type=Bone,Name="deform_collar_b_l"))
   )
   n3 = &RigUnit_GetTransform(Item=(Type=Control,Name="head_ctrl"), bInitial=False)
   n4 = &RigUnit_GetTransform(Item=(Type=Bone,Name="dyn_collar_bk"), bInitial=False)
   > &RigUnit_ModifyTransforms(
    ItemToModify="((Item=(Type=Bone,Name=\"deform_collar_bk\"),Transform=(Rotation=(X=0,Y=0,Z=-0.2164,W=0.9763),Translation=(X=0,Y=-2,Z=0),Scale3D=(X=1,Y=1,Z=1.3))),(Item=(Type=Bone,Name=\"deform_collar_b_r\"),Transform=(Rotation=(X=0,Y=0,Z=-0.1736,W=0.9848),Translation=(X=0,Y= ...(truncated)"
    WeightMinimum=0
    Weight=&RigVMFunction_MathDoubleRemap(
     SourceMaximum=0
     SourceMinimum=12
     TargetMaximum=1
     TargetMinimum=0
     bClamp=False
     Value=&RigVMFunction_MathVectorDistance(A=@n3.Transform.Translation, B=@n4.Transform.Translation).Result
    ).Result
   )
   n5 = &RigUnit_GetTransform(Item=(Type=Control,Name="head_ctrl"), bInitial=False)
   n6 = &RigUnit_GetTransform(Item=(Type=Bone,Name="dyn_collar_b_r"), bInitial=False)
   > &RigUnit_ModifyTransforms(
    ItemToModify="((Item=(Type=Bone,Name=\"deform_collar_b_r\"),Transform=(Rotation=(X=0,Y=0,Z=-0.6756,W=0.7373),Translation=(X=0,Y=-5,Z=0),Scale3D=(X=1,Y=1,Z=1))),(Item=(Type=Bone,Name=\"deform_collar_bk\"),Transform=(Rotation=(X=0.2091,Y=-0.056,Z=-0.2527,W=0.943),Translation= ...(truncated)"
    WeightMinimum=0
    Weight=&RigVMFunction_MathDoubleRemap(
     SourceMaximum=0
     SourceMinimum=12
     TargetMaximum=1
     TargetMinimum=0
     bClamp=False
     Value=&RigVMFunction_MathVectorDistance(A=@n5.Transform.Translation, B=@n6.Transform.Translation).Result
    ).Result
   )
   n7 = &RigUnit_GetTransform(Item=(Type=Control,Name="head_ctrl"), bInitial=False)
   n8 = &RigUnit_GetTransform(Item=(Type=Bone,Name="dyn_collar_b_l"), bInitial=False)
   > &RigUnit_ModifyTransforms(
    ItemToModify="((Item=(Type=Bone,Name=\"deform_collar_b_l\"),Transform=(Rotation=(X=0,Y=0,Z=-0.5,W=0.866),Translation=(X=0,Y=-2,Z=0),Scale3D=(X=1,Y=1,Z=1))),(Item=(Type=Bone,Name=\"deform_collar_bk\"),Transform=(Rotation=(X=0,Y=0,Z=-0.2588,W=0.9659),Translation=(X=0,Y=-1,Z=0 ...(truncated)"
    WeightMinimum=0
    Weight=&RigVMFunction_MathDoubleRemap(
     SourceMaximum=0
     SourceMinimum=12
     TargetMaximum=1
     TargetMinimum=0
     bClamp=False
     Value=&RigVMFunction_MathVectorDistance(A=@n7.Transform.Translation, B=@n8.Transform.Translation).Result
    ).Result
   )
   RigUnit_GetRelativeTransformForItem = &RigUnit_GetRelativeTransformForItem(Child=(Type=Bone,Name="head"), Parent=(Type=Bone,Name="neck_01"), bChildInitial=False, bParentInitial=False)
   RigVMFunction_MathQuaternionToEuler = &RigVMFunction_MathQuaternionToEuler(RotationOrder=ZYX, Value=@RigUnit_GetRelativeTransformForItem.RelativeTransform.Rotation)
   > &RigUnit_ModifyTransforms(
    ItemToModify="((Item=(Type=Bone,Name=\"deform_collar_b_r\"),Transform=(Rotation=(X=0,Y=0,Z=-0.1736,W=0.9848),Translation=(X=0,Y=0,Z=0),Scale3D=(X=1,Y=1,Z=1))),(Item=(Type=Bone,Name=\"deform_collar_bk\"),Transform=(Rotation=(X=0,Y=0,Z=-0.0872,W=0.9962),Translation=(X=0,Y=0,Z ...(truncated)"
    WeightMinimum=0
    Weight=&RigVMFunction_MathDoubleRemap(SourceMaximum=50, SourceMinimum=0, TargetMaximum=1, TargetMinimum=0, Value=@RigVMFunction_MathQuaternionToEuler.Result.X).Result
   )
   > &RigUnit_ModifyTransforms(
    ItemToModify="((Item=(Type=Bone,Name=\"deform_collar_b_l\"),Transform=(Rotation=(X=0,Y=0,Z=-0.1736,W=0.9848),Translation=(X=0,Y=0,Z=0),Scale3D=(X=1,Y=1,Z=1))),(Item=(Type=Bone,Name=\"deform_collar_bk\"),Transform=(Rotation=(X=0,Y=0,Z=-0.0872,W=0.9962),Translation=(X=0,Y=0,Z ...(truncated)"
    WeightMinimum=0
    Weight=&RigVMFunction_MathDoubleRemap(SourceMaximum=-50, SourceMinimum=0, TargetMaximum=1, TargetMinimum=0, Value=@RigVMFunction_MathQuaternionToEuler.Result.X).Result
   )
   RigVMFunction_ControlFlowBranch = &RigVMFunction_ControlFlowBranch(ExecuteContext=@RigVMFunction_Sequence.E, Condition=RigVMVariableNode() {variable=RBAN}.Value)
   &RigUnit_AimConstraintLocalSpaceOffset(
    Child=(Type=Bone,Name="dyn_skirt_b_1_r")
    Filter=(bX=True,bY=True,bZ=True)
    Parents=((Item=(Type=Bone,Name="foot_r"),Weight=1))
    UpAxis=(X=0,Y=-1,Z=0)
    Weight=1
    WorldUp=(Target=(X=0,Y=0,Z=1),Kind=Location,Space=(Type=Bone,Name="deform_skirt_b_r"))
    ExecutePin=@RigVMFunction_ControlFlowBranch.False
   )
   > &RigUnit_AimConstraintLocalSpaceOffset(
    Child=(Type=Bone,Name="dyn_skirt_b_1_l")
    Filter=(bX=True,bY=True,bZ=True)
    Parents=((Item=(Type=Bone,Name="foot_l"),Weight=1))
    UpAxis=(X=0,Y=-1,Z=0)
    Weight=1
    WorldUp=(Target=(X=0,Y=0,Z=1),Kind=Location,Space=(Type=Bone,Name="deform_skirt_b_l"))
   )
  ]}
]
hierarchy [
 root {loc=<0 0 0>, rot=<0 0 0 1>}
 pelvis {parent=root, loc=<0 -0.7071 93.1191>, rot=<0.0127 0.707 -0.0127 -0.707>}
 spine_01 {parent=pelvis, loc=<6.7494 -0.6388 0>, rot=<0 0 0.034 -0.9994>}
 spine_02 {parent=spine_01, loc=<7.252 -0.035 0>, rot=<0 0 0.0238 -0.9997>}
 spine_03 {parent=spine_02, loc=<6.0106 0 0>, rot=<0 0 0.0149 -0.9999>}
 spine_04 {parent=spine_03, loc=<8.1023 0 0>, rot=<0 0 -0.0132 -0.9999>}
 spine_05 {parent=spine_04, loc=<10.0142 0 0>, rot=<0 0 -0.0015 -1>}
 attach_backpack {parent=spine_05, loc=<-0.123 -13.473 0>, rot=<-0.0093 -0.707 0.0093 -0.707>}
 neck_01 {parent=spine_05, loc=<11.3138 0.1529 0>, rot=<0 0 -0.2003 -0.9797>}
 neck_02 {parent=neck_01, loc=<4.5627 0 0>, rot=<0 0 0.0504 -0.9987>}
 head {parent=neck_02, loc=<5.3914 0 0>, rot=<0 0 0.115 -0.9934>}
 head_ctrl {controlType=EULER_TRANSFORM, parent=head, type=Control, loc=<7 -7 0>, rot=<0 0 0 1>, scale=<0.2 0.2 0.2>}
 deform_collar_f_l {parent=neck_01, loc=<0.7628 6.6666 -4.8397>, rot=<0.9089 0.0399 0.005 0.415>}
 dyn_collar_f_l {parent=deform_collar_f_l, loc=<0 0 0>, rot=<0 0 0 1>}
 deform_collar_l {parent=neck_01, loc=<1.5708 1.6947 -8.293>, rot=<-0.7434 0.0496 -0.0868 -0.6614>}
 dyn_collar_l {parent=deform_collar_l, loc=<0 0 0>, rot=<0 0 0 1>}
 deform_collar_b_l {parent=neck_01, loc=<1.9244 -3.8815 -5.7461>, rot=<-0.3908 0.0083 0.0393 -0.9196>}
 dyn_collar_b_l {parent=deform_collar_b_l, loc=<0 0 0>, rot=<0 0 0 1>}
 deform_collar_bk {parent=neck_01, loc=<1.9264 -5.9281 0>, rot=<0 0 0 1>}
 dyn_collar_bk {parent=deform_collar_bk, loc=<0 0 0>, rot=<0 0 0 1>}
 deform_collar_f_r {parent=neck_01, loc=<0.7632 6.6664 4.8397>, rot=<0.9089 0.0399 -0.005 -0.415>}
 dyn_collar_f_r {parent=deform_collar_f_r, loc=<0 0 0>, rot=<0 0 0 -1>}
 deform_collar_r {parent=neck_01, loc=<1.5707 1.695 8.293>, rot=<0.7434 -0.0496 -0.0868 -0.6614>}
 dyn_collar_r {parent=deform_collar_r, loc=<0 0 0>, rot=<0 0 0 -1>}
 deform_collar_b_r {parent=neck_01, loc=<1.9246 -3.8816 5.7461>, rot=<0.3908 -0.0083 0.0393 -0.9196>}
 dyn_collar_b_r {parent=deform_collar_b_r, loc=<0 0 0>, rot=<0 0 0 -1>}
 clavicle_l {parent=spine_05, loc=<4.3575 -0.1466 -2.3223>, rot=<0.0465 -0.7289 -0.009 -0.6829>}
 upperarm_l {parent=clavicle_l, loc=<15.5111 0 0>, rot=<0.0226 -0.3234 -0.0088 -0.9459>}
 lowerarm_l {parent=upperarm_l, loc=<26.6514 0 0>, rot=<0.0391 -0.0099 -0.2453 -0.9686>}
 hand_l {parent=lowerarm_l, loc=<22.3491 0 0>, rot=<-0.7081 -0.0507 0.0975 -0.6975>}
 thumb_01_l {parent=hand_l, loc=<1.3118 -0.3191 -2.459>, rot=<0.6546 -0.4042 0.0664 -0.6354>}
 thumb_02_l {parent=thumb_01_l, loc=<5.9336 -0.0448 0.0401>, rot=<-0.0037 0.0213 0.298 -0.9543>}
 thumb_03_l {parent=thumb_02_l, loc=<2.2223 0.3682 -0.0907>, rot=<0.0118 -0.013 0.0066 -0.9998>}
 index_metacarpal_l {parent=hand_l, loc=<1.6535 0.0183 -2.96>, rot=<0.0158 -0.0124 -0.0173 -0.9996>}
 index_01_l {parent=index_metacarpal_l, loc=<8.3008 0.0357 -0.0524>, rot=<-0.0044 0.0485 0.0924 -0.9945>}
 index_02_l {parent=index_01_l, loc=<3.3587 -0.3004 0.0452>, rot=<-0.002 0.0027 0.3225 -0.9466>}
 index_03_l {parent=index_02_l, loc=<2.5578 -0.123 -0.0252>, rot=<-0.002 0.0067 0.2558 -0.9667>}
 middle_metacarpal_l {parent=hand_l, loc=<2.1856 -0.1725 -0.5984>, rot=<0.006 0.0168 -0.0329 -0.9993>}
 middle_01_l {parent=middle_metacarpal_l, loc=<7.3275 -0.2447 0.0129>, rot=<-0.0208 0.0042 0.0858 -0.9961>}
 middle_02_l {parent=middle_01_l, loc=<4.0123 0.1401 0.1175>, rot=<-0.0218 0.0258 0.3512 -0.9357>}
 middle_03_l {parent=middle_02_l, loc=<2.6077 0.0534 0.1373>, rot=<0.0082 0.0188 0.1708 -0.9851>}
 ring_metacarpal_l {parent=hand_l, loc=<2.2074 -0.3625 1.3638>, rot=<-0.0284 0.0722 -0.0458 -0.9959>}
 ring_01_l {parent=ring_metacarpal_l, loc=<7.0295 -0.6197 -0.1446>, rot=<-0.0489 -0.0252 0.1892 -0.9804>}
 ring_02_l {parent=ring_01_l, loc=<3.6174 0.4647 -0.1027>, rot=<0 0 0.2806 -0.9598>}
 ring_03_l {parent=ring_02_l, loc=<2.4152 0.114 0.0199>, rot=<0 0 0.1189 -0.9929>}
 pinky_metacarpal_l {parent=hand_l, loc=<1.7616 -0.9753 3.3814>, rot=<-0.0397 0.0814 -0.069 -0.9935>}
 pinky_01_l {parent=pinky_metacarpal_l, loc=<7.1709 -0.4831 -0.1094>, rot=<-0.0654 -0.0035 0.2242 -0.9723>}
 pinky_02_l {parent=pinky_01_l, loc=<2.8274 0.0893 -0.1043>, rot=<0.0037 0.0011 0.2854 -0.9584>}
 pinky_03_l {parent=pinky_02_l, loc=<2.2046 0.0307 -0.0078>, rot=<-0.0002 0.0098 0.1445 -0.9895>}
 deform_wrist_dn_base_l {parent=hand_l, loc=<0 -2.2342 0>, rot=<0 0 0 -1>}
 deform_wrist_dn_l {parent=deform_wrist_dn_base_l, loc=<0 0 0>, rot=<0 0 0 1>}
 deform_wrist_up_base_l {parent=hand_l, loc=<0 2.2682 0>, rot=<0 0 0 -1>}
 deform_wrist_up_l {parent=deform_wrist_up_base_l, loc=<0 0 0>, rot=<0 0 0 1>}
 hand_attach_l {parent=hand_l, loc=<8.5 -3 0>, rot=<0.7071 0.7071 0 0>}
 weapon_l {parent=hand_l, loc=<-0.1974 -3.4111 0.3811>, rot=<0 0 0 1>}
 lowerarm_twist_02_l {parent=lowerarm_l, loc=<9.6199 0 0>, rot=<0 0 0 -1>}
 lowerarm_twist_01_l {parent=lowerarm_l, loc=<17.8549 0 0>, rot=<0 0 0 -1>}
 deform_elbow_in_base_l {parent=lowerarm_l, loc=<0.2112 5.259 0.8274>, rot=<0 0 0 -1>}
 deform_elbow_in_l {parent=deform_elbow_in_base_l, loc=<0 0 0>, rot=<0 0 0 1>}
 deform_elbow_out_base_l {parent=lowerarm_l, loc=<0 -4 0>, rot=<0 0 0 -1>}
 deform_elbow_out_l {parent=deform_elbow_out_base_l, loc=<0 0 0>, rot=<0 0 0 1>}
 upperarm_twist_01_l {parent=upperarm_l, loc=<0.0041 0 0>, rot=<0 0 0 -1>}
 deform_upperarm_bk_base_l {parent=upperarm_twist_01_l, loc=<1.0813 -7.2193 -1.9453>, rot=<0 0 0 1>}
 deform_upperarm_bk_l {parent=deform_upperarm_bk_base_l, loc=<0 0 0>, rot=<0 0 0 1>}
 deform_upperarm_fr_base_l {parent=upperarm_twist_01_l, loc=<-0.3691 7.0506 -2.765>, rot=<0 0 0 1>}
 deform_upperarm_fr_l {parent=deform_upperarm_fr_base_l, loc=<0 0 0>, rot=<0 0 0 1>}
 deform_upperarm_up_base_l {parent=upperarm_twist_01_l, loc=<-3.2216 -1.1391 3.3311>, rot=<0 0 0 1>}
 deform_upperarm_up_l {parent=deform_upperarm_up_base_l, loc=<0 0 0>, rot=<0 0 0 1>}
 upperarm_twist_02_l {parent=upperarm_l, loc=<13.7941 0 0>, rot=<0 0 0 -1>}
 attach_cape {parent=spine_05, loc=<9.017 -11.714 0>, rot=<0.0616 -0.7044 -0.0616 -0.7044>}
 clavicle_r {parent=spine_05, loc=<4.3575 -0.1466 2.3223>, rot=<0.7289 0.0465 -0.6829 0.009>}
 upperarm_r {parent=clavicle_r, loc=<-15.5111 0 0>, rot=<0.0226 -0.3234 -0.0088 -0.9459>}
 lowerarm_r {parent=upperarm_r, loc=<-26.6514 0 0>, rot=<0.0391 -0.0099 -0.2453 -0.9686>}
 hand_r {parent=lowerarm_r, loc=<-22.3491 0 0>, rot=<-0.7081 -0.0507 0.0975 -0.6975>}
 thumb_01_r {parent=hand_r, loc=<-1.3118 0.3191 2.459>, rot=<0.6546 -0.4042 0.0664 -0.6354>}
 thumb_02_r {parent=thumb_01_r, loc=<-5.9336 0.0448 -0.0401>, rot=<-0.0037 0.0213 0.298 -0.9543>}
 thumb_03_r {parent=thumb_02_r, loc=<-2.2223 -0.3682 0.0907>, rot=<0.0118 -0.013 0.0066 -0.9998>}
 index_metacarpal_r {parent=hand_r, loc=<-1.6535 -0.0183 2.96>, rot=<0.0158 -0.0124 -0.0173 -0.9996>}
 index_01_r {parent=index_metacarpal_r, loc=<-8.3008 -0.0357 0.0524>, rot=<-0.0044 0.0485 0.0924 -0.9945>}
 index_02_r {parent=index_01_r, loc=<-3.3587 0.3004 -0.0452>, rot=<-0.002 0.0027 0.3225 -0.9466>}
 index_03_r {parent=index_02_r, loc=<-2.5578 0.123 0.0252>, rot=<-0.002 0.0067 0.2558 -0.9667>}
 middle_metacarpal_r {parent=hand_r, loc=<-2.1856 0.1725 0.5984>, rot=<0.006 0.0168 -0.0329 -0.9993>}
 middle_01_r {parent=middle_metacarpal_r, loc=<-7.3275 0.2447 -0.0129>, rot=<-0.0208 0.0042 0.0858 -0.9961>}
 middle_02_r {parent=middle_01_r, loc=<-4.0123 -0.1401 -0.1175>, rot=<-0.0218 0.0258 0.3512 -0.9357>}
 middle_03_r {parent=middle_02_r, loc=<-2.6077 -0.0534 -0.1373>, rot=<0.0082 0.0188 0.1708 -0.9851>}
 ring_metacarpal_r {parent=hand_r, loc=<-2.2074 0.3625 -1.3638>, rot=<-0.0284 0.0722 -0.0458 -0.9959>}
 ring_01_r {parent=ring_metacarpal_r, loc=<-7.0295 0.6197 0.1446>, rot=<-0.0489 -0.0252 0.1892 -0.9804>}
 ring_02_r {parent=ring_01_r, loc=<-3.6174 -0.4647 0.1027>, rot=<0 0 0.2806 -0.9598>}
 ring_03_r {parent=ring_02_r, loc=<-2.4152 -0.114 -0.0199>, rot=<0 0 0.1189 -0.9929>}
 pinky_metacarpal_r {parent=hand_r, loc=<-1.7616 0.9753 -3.3814>, rot=<-0.0397 0.0814 -0.069 -0.9935>}
 pinky_01_r {parent=pinky_metacarpal_r, loc=<-7.1709 0.4831 0.1094>, rot=<-0.0654 -0.0035 0.2242 -0.9723>}
 pinky_02_r {parent=pinky_01_r, loc=<-2.8274 -0.0893 0.1043>, rot=<0.0037 0.0011 0.2854 -0.9584>}
 pinky_03_r {parent=pinky_02_r, loc=<-2.2046 -0.0307 0.0078>, rot=<-0.0002 0.0098 0.1445 -0.9895>}
 deform_wrist_dn_base_r {parent=hand_r, loc=<-0.0004 2.2345 -0.0001>, rot=<0 0 0 -1>}
 deform_wrist_dn_r {parent=deform_wrist_dn_base_r, loc=<0 0 0>, rot=<0 0 0 1>}
 deform_wrist_up_base_r {parent=hand_r, loc=<0.0003 -2.2684 0.0001>, rot=<0 0 0 -1>}
 deform_wrist_up_r {parent=deform_wrist_up_base_r, loc=<0 0 0>, rot=<0 0 0 1>}
 hand_attach_r {parent=hand_r, loc=<-8.5001 2.9999 0>, rot=<0 0 -0.7071 -0.7071>}
 lowerarm_twist_02_r {parent=lowerarm_r, loc=<-9.6199 0 0>, rot=<0 0 0 -1>}
 lowerarm_twist_01_r {parent=lowerarm_r, loc=<-17.8549 0 0>, rot=<0 0 0 -1>}
 deform_elbow_in_base_r {parent=lowerarm_r, loc=<-0.2111 -5.2591 -0.8275>, rot=<0 0 0 -1>}
 deform_elbow_in_r {parent=deform_elbow_in_base_r, loc=<-0.0001 0 0.0002>, rot=<0 0 0 1>}
 deform_elbow_out_base_r {parent=lowerarm_r, loc=<0 4 -0.0001>, rot=<0 0 0 -1>}
 deform_elbow_out_r {parent=deform_elbow_out_base_r, loc=<0 0 0>, rot=<0 0 0 1>}
 upperarm_twist_01_r {parent=upperarm_r, loc=<-0.0041 0 0>, rot=<0 0 0 -1>}
 deform_upperarm_bk_base_r {parent=upperarm_twist_01_r, loc=<-1.0814 7.2192 1.9455>, rot=<0 0 0 1>}
 deform_upperarm_bk_r {parent=deform_upperarm_bk_base_r, loc=<0 0 0>, rot=<0 0 0 1>}
 deform_upperarm_fr_base_r {parent=upperarm_twist_01_r, loc=<0.3689 -7.0506 2.7652>, rot=<0 0 0 1>}
 deform_upperarm_fr_r {parent=deform_upperarm_fr_base_r, loc=<0 0 0>, rot=<0 0 0 1>}
 deform_upperarm_up_base_r {parent=upperarm_twist_01_r, loc=<3.2214 1.1391 -3.3308>, rot=<0 0 0 1>}
 deform_upperarm_up_r {parent=deform_upperarm_up_base_r, loc=<0 0 0>, rot=<0 0 0 -1>}
 upperarm_twist_02_r {parent=upperarm_r, loc=<-13.7941 0 0>, rot=<0 0 0 -1>}
 deform_chain_up_inn_l {parent=spine_05, loc=<9.0475 8.205 -6.178>, rot=<-0.0696 0.0801 0.9324 0.3454>}
 dyn_chain_up_inn_l {parent=deform_chain_up_inn_l, loc=<0 0 0>, rot=<0 0 0 1>}
 dyn_chain_up_inn_1_l {parent=dyn_chain_up_inn_l, loc=<10.2681 2.1887 -4.7223>, rot=<-0.0694 -0.621 -0.1137 -0.7724>}
 deform_chain_up_out_l {parent=spine_05, loc=<5.7367 6.5277 -17.6756>, rot=<0.0503 0.1374 0.9644 0.2203>}
 dyn_chain_up_out_l {parent=deform_chain_up_out_l, loc=<0 0 0>, rot=<0 0 0 1>}
 dyn_chain_down_out_1_l {parent=dyn_chain_up_out_l, loc=<7.9831 0.7464 0.358>, rot=<0.0037 0.122 -0.0298 -0.9921>}
 dyn_chain_down_out_2_l {parent=dyn_chain_down_out_1_l, loc=<7.4869 -1.4851 5.2735>, rot=<-0.0768 0.6206 0.0982 -0.7742>}
 dyn_chain_up_out_1_l {parent=dyn_chain_up_out_l, loc=<7.5791 -1.0981 5.4752>, rot=<0.0551 0.7287 0.1009 -0.6751>}
 deform_chain_down_inn_l {parent=spine_05, loc=<7.6523 9.6606 -2.533>, rot=<-0.092 0.0904 0.9424 0.3088>}
 dyn_chain_down_inn_l {parent=deform_chain_down_inn_l, loc=<0 0 0>, rot=<0 0 0 1>}
 dyn_chain_down_inn_1_l {parent=dyn_chain_down_inn_l, loc=<7.5335 0 0>, rot=<-0.017 -0.0761 -0.0962 -0.9923>}
 dyn_chain_down_inn_2_l {parent=dyn_chain_down_inn_1_l, loc=<8.659 1.3311 -3.5865>, rot=<-0.07 -0.542 -0.1437 -0.825>}
 deform_chain_up_inn_r {parent=spine_05, loc=<9.0476 8.205 6.178>, rot=<0.0696 -0.0801 0.9324 0.3454>}
 dyn_chain_up_inn_r {parent=deform_chain_up_inn_r, loc=<0 0 0>, rot=<0 0 0 1>}
 dyn_chain_up_inn_1_r {parent=dyn_chain_up_inn_r, loc=<10.2681 2.1888 4.7224>, rot=<0.0694 0.621 -0.1137 -0.7724>}
 deform_chain_up_out_r {parent=spine_05, loc=<5.7363 6.5277 17.6756>, rot=<-0.0503 -0.1374 0.9644 0.2203>}
 dyn_chain_up_out_r {parent=deform_chain_up_out_r, loc=<0 0 0>, rot=<0 0 0 -1>}
 dyn_chain_up_out_1_r {parent=dyn_chain_up_out_r, loc=<7.5788 -1.0983 -5.4753>, rot=<-0.0551 -0.7287 0.1009 -0.6751>}
 dyn_chain_down_out_1_r {parent=dyn_chain_up_out_r, loc=<7.983 0.7463 -0.358>, rot=<-0.0037 -0.122 -0.0298 -0.9921>}
 dyn_chain_down_out_2_r {parent=dyn_chain_down_out_1_r, loc=<7.4867 -1.4852 -5.2735>, rot=<0.0768 -0.6206 0.0982 -0.7742>}
 deform_chain_down_inn_r {parent=spine_05, loc=<7.6525 9.6606 2.533>, rot=<0.092 -0.0904 0.9424 0.3088>}
 dyn_chain_down_inn_r {parent=deform_chain_down_inn_r, loc=<0 0 0>, rot=<0 0 0 1>}
 dyn_chain_down_inn_1_r {parent=dyn_chain_down_inn_r, loc=<7.5333 -0.0002 0>, rot=<0.017 0.0761 -0.0962 -0.9923>}
 dyn_chain_down_inn_2_r {parent=dyn_chain_down_inn_1_r, loc=<8.6594 1.3313 3.5864>, rot=<0.07 0.542 -0.1437 -0.825>}
 deform_pec_aim_base_l {parent=spine_04, loc=<7.8728 12.6388 -2.4123>, rot=<0.0102 -0.4602 0.3979 -0.7936>}
 deform_pec_aim_l {parent=deform_pec_aim_base_l, loc=<0 0 0>, rot=<0 0 0 1>}
 deform_pec_base_l {parent=deform_pec_aim_l, loc=<10.766 0 0>, rot=<0 0 0 1>}
 deform_pec_l {parent=deform_pec_base_l, loc=<0.055 0 0>, rot=<0 0 0 1>}
 deform_pec_aim_base_r {parent=spine_04, loc=<7.8723 12.6388 2.4123>, rot=<-0.4602 -0.0102 0.7936 0.3979>}
 deform_pec_aim_r {parent=deform_pec_aim_base_r, loc=<0 0 0>, rot=<0 0 0 1>}
 deform_pec_base_r {parent=deform_pec_aim_r, loc=<-10.7662 -0.0006 -0.0007>, rot=<0 0 0 1>}
 deform_pec_r {parent=deform_pec_base_r, loc=<0 0 0>, rot=<0 0 0 1>}
 deform_skirt_f_l {parent=spine_01, loc=<-2.8698 8.4685 -8.9012>, rot=<-0.1008 0.3362 0.9328 0.0822>}
 dyn_skirt_f_l {parent=deform_skirt_f_l, loc=<0 0 0>, rot=<0 0 0 1>}
 dyn_skirt_f_1_l {parent=dyn_skirt_f_l, loc=<12.693 0 0>, rot=<0 0 -0.0339 -0.9994>}
 dyn_skirt_f_2_l {parent=dyn_skirt_f_1_l, loc=<18.5654 0 0>, rot=<0 0 -0.0145 -0.9999>}
 dyn_skirt_f_3_l {parent=dyn_skirt_f_2_l, loc=<22.2258 0 0>, rot=<0 0 0 1>}
 deform_skirt_m_l {parent=spine_01, loc=<-1.9946 3.4844 -13.2063>, rot=<-0.1357 0.5543 0.8196 0.0508>}
 dyn_skirt_m_l {parent=deform_skirt_m_l, loc=<0 0 0>, rot=<0 0 0 1>}
 dyn_skirt_m_1_l {parent=dyn_skirt_m_l, loc=<12.693 0 0>, rot=<0 0 -0.0137 -0.9999>}
 dyn_skirt_m_2_l {parent=dyn_skirt_m_1_l, loc=<18.5654 0 0>, rot=<0 0 0.002 -1>}
 dyn_skirt_m_3_l {parent=dyn_skirt_m_2_l, loc=<22.2258 0 0>, rot=<0 0 -0.0277 -0.9996>}
 deform_skirt_b_l {parent=spine_01, loc=<-0.6013 -3.6877 -11.7569>, rot=<0.13 -0.852 -0.5054 -0.041>}
 dyn_skirt_b_l {parent=deform_skirt_b_l, loc=<0 0 0>, rot=<0 0 0 1>}
 dyn_skirt_b_1_l {parent=dyn_skirt_b_l, loc=<15.5988 0.0048 0>, rot=<0 0 -0.0008 -1>}
 dyn_skirt_b_2_l {parent=dyn_skirt_b_1_l, loc=<16.7987 0.0071 0>, rot=<0 0 0.002 -1>}
 dyn_skirt_b_3_l {parent=dyn_skirt_b_2_l, loc=<21.4763 -0.0415 0>, rot=<0 0 -0.0277 -0.9996>}
 deform_skirt_bk_l {parent=spine_01, loc=<-0.2678 -8.9201 -5.1085>, rot=<0.1179 -0.9825 -0.1423 -0.0217>}
 dyn_skirt_bk_l {parent=deform_skirt_bk_l, loc=<0 0 0>, rot=<0 0 0 1>}
 dyn_skirt_bk_1_l {parent=dyn_skirt_bk_l, loc=<15.9774 -0.0516 0>, rot=<0 0 0.0076 -1>}
 dyn_skirt_bk_2_l {parent=dyn_skirt_bk_1_l, loc=<17.6774 0.0036 0>, rot=<0 0 0.002 -1>}
 dyn_skirt_bk_3_l {parent=dyn_skirt_bk_2_l, loc=<22.4299 0.0113 0>, rot=<0 0 -0.0277 -0.9996>}
 deform_skirt_f_r {parent=spine_01, loc=<-2.8698 8.4685 8.9012>, rot=<0.1008 -0.3362 0.9328 0.0822>}
 dyn_skirt_f_r {parent=deform_skirt_f_r, loc=<0 0 0>, rot=<0 0 0 1>}
 dyn_skirt_f_1_r {parent=dyn_skirt_f_r, loc=<12.6931 0 0>, rot=<0 0 -0.0339 -0.9994>}
 dyn_skirt_f_2_r {parent=dyn_skirt_f_1_r, loc=<18.5654 0 0>, rot=<0 0 -0.0145 -0.9999>}
 dyn_skirt_f_3_r {parent=dyn_skirt_f_2_r, loc=<22.2258 0 0>, rot=<0 0 0 1>}
 deform_skirt_m_r {parent=spine_01, loc=<-1.9946 3.4844 13.2063>, rot=<0.1357 -0.5543 0.8196 0.0508>}
 dyn_skirt_m_r {parent=deform_skirt_m_r, loc=<0 0 0>, rot=<0 0 0 1>}
 dyn_skirt_m_1_r {parent=dyn_skirt_m_r, loc=<12.693 0 0>, rot=<0 0 -0.0137 -0.9999>}
 dyn_skirt_m_2_r {parent=dyn_skirt_m_1_r, loc=<18.5654 -0.0001 0>, rot=<0 0 0.002 -1>}
 dyn_skirt_m_3_r {parent=dyn_skirt_m_2_r, loc=<22.2258 0 0>, rot=<0 0 -0.0277 -0.9996>}
 deform_skirt_b_r {parent=spine_01, loc=<-0.6013 -3.6877 11.7569>, rot=<0.13 -0.852 0.5054 0.041>}
 dyn_skirt_b_r {parent=deform_skirt_b_r, loc=<0 0 0>, rot=<0 0 0 -1>}
 dyn_skirt_b_1_r {parent=dyn_skirt_b_r, loc=<15.5988 0.0048 0>, rot=<0 0 -0.0008 -1>}
 dyn_skirt_b_2_r {parent=dyn_skirt_b_1_r, loc=<16.7987 0.0071 0>, rot=<0 0 0.002 -1>}
 dyn_skirt_b_3_r {parent=dyn_skirt_b_2_r, loc=<21.4764 -0.0415 0>, rot=<0 0 -0.0277 -0.9996>}
 deform_skirt_bk_r {parent=spine_01, loc=<-0.2678 -8.9201 5.1085>, rot=<0.1179 -0.9825 0.1423 0.0217>}
 dyn_skirt_bk_r {parent=deform_skirt_bk_r, loc=<0 0 0>, rot=<0 0 0 1>}
 dyn_skirt_bk_1_r {parent=dyn_skirt_bk_r, loc=<15.9774 -0.0516 0>, rot=<0 0 0.0076 -1>}
 dyn_skirt_bk_2_r {parent=dyn_skirt_bk_1_r, loc=<17.6774 0.0036 0>, rot=<0 0 0.002 -1>}
 dyn_skirt_bk_3_r {parent=dyn_skirt_bk_2_r, loc=<22.43 0.0113 0>, rot=<0 0 -0.0277 -0.9996>}
 thigh_l {parent=pelvis, loc=<-2.5693 -0.0109 -7.7501>, rot=<0.0012 0.0572 0.0155 -0.9982>}
 calf_l {parent=thigh_l, loc=<-39.0484 0 0>, rot=<0.0006 -0.0161 -0.0391 -0.9991>}
 foot_l {parent=calf_l, loc=<-42.561 0.1077 0.4292>, rot=<0.0418 -0.0455 0.0343 -0.9975>}
 ball_l {parent=foot_l, loc=<-6.1599 13.8264 0.6806>, rot=<0.0001 -0.0001 -0.7063 -0.7079>}
 calf_twist_02_l {parent=calf_l, loc=<-18.8064 -0.0967 0.0397>, rot=<0.0237 0.0095 -0.0035 -0.9997>}
 calf_twist_01_l {parent=calf_l, loc=<-35.813 -0.0967 0.0397>, rot=<0.0237 0.0095 -0.0035 -0.9997>}
 deform_knee_in_base_l {parent=calf_l, loc=<1.298 -5.2957 -0.0496>, rot=<0 0 0 -1>}
 deform_knee_in_l {parent=deform_knee_in_base_l, loc=<0 0 0>, rot=<0 0 0 1>}
 deform_knee_out_base_l {parent=calf_l, loc=<0 4 0>, rot=<0 0 0 -1>}
 deform_knee_out_l {parent=deform_knee_out_base_l, loc=<0 0 0>, rot=<0 0 0 1>}
 thigh_twist_01_l {parent=thigh_l, loc=<-21.6303 0 0>, rot=<0.0013 0 0 -1>}
 deform_groin_base_l {parent=thigh_l, loc=<-2.1908 8.1142 -1.7442>, rot=<0 0 0 -1>}
 deform_groin_l {parent=deform_groin_base_l, loc=<0 0 0>, rot=<0 0 0 1>}
 thigh_r {parent=pelvis, loc=<-2.5693 -0.0109 7.7501>, rot=<0.0572 -0.0012 0.9982 0.0155>}
 calf_r {parent=thigh_r, loc=<39.0484 0 0>, rot=<0.0006 -0.0161 -0.0391 -0.9991>}
 foot_r {parent=calf_r, loc=<42.561 -0.1077 -0.4292>, rot=<0.0418 -0.0455 0.0343 -0.9975>}
 ball_r {parent=foot_r, loc=<6.1599 -13.8264 -0.6806>, rot=<0.0001 -0.0001 -0.7063 -0.7079>}
 calf_twist_02_r {parent=calf_r, loc=<18.8064 0.0967 -0.0397>, rot=<0.0237 0.0095 -0.0035 -0.9997>}
 calf_twist_01_r {parent=calf_r, loc=<35.813 0.0967 -0.0397>, rot=<0.0237 0.0095 -0.0035 -0.9997>}
 deform_knee_in_base_r {parent=calf_r, loc=<-1.298 5.2957 0.0496>, rot=<0 0 0 -1>}
 deform_knee_in_r {parent=deform_knee_in_base_r, loc=<0 0 0>, rot=<0 0 0 1>}
 deform_knee_out_base_r {parent=calf_r, loc=<0 -4 0>, rot=<0 0 0 -1>}
 deform_knee_out_r {parent=deform_knee_out_base_r, loc=<0 0 0>, rot=<0 0 0 1>}
 thigh_twist_01_r {parent=thigh_r, loc=<21.6303 0 0>, rot=<0.0013 0 0 -1>}
 deform_groin_base_r {parent=thigh_r, loc=<2.1908 -8.1142 1.7442>, rot=<0 0 0 -1>}
 deform_groin_r {parent=deform_groin_base_r, loc=<0 0 0>, rot=<0 0 0 1>}
 deform_glute_base_l {parent=pelvis, loc=<-8.8064 -8.8222 -8.4287>, rot=<0.0275 0.0501 -0.4472 -0.8926>}
 deform_glute_l {parent=deform_glute_base_l, loc=<0 0 0>, rot=<0 0 0 1>}
 deform_glute_base_r {parent=pelvis, loc=<-8.8062 -8.8222 8.4287>, rot=<0.0501 -0.0275 0.8926 -0.4472>}
 deform_glute_r {parent=deform_glute_base_r, loc=<0 0 0>, rot=<0 0 0 1>}
 thigh_r_ctrl {controlType=EULER_TRANSFORM, parent=pelvis, type=Control, loc=<0 0 0>, rot=<0 0 0 1>}
 thigh_l_ctrl {controlType=EULER_TRANSFORM, parent=pelvis, type=Control, loc=<0 0 0>, rot=<0 0 0 1>}
 deform_belt {parent=pelvis, loc=<0 0 0>, rot=<0 0 0 1>}
 deform_belt1 {parent=deform_belt, loc=<0 0 0>, rot=<0 0 0 1>}
]
variables [
 DBNO {type=bool, default=False}
 IsEmote {type=bool, default=False}
 KneeSliding {type=bool, default=False}
 RBAN {type=bool, default=False}
 Skydiving {type=bool, default=False}
 Sliding {type=bool, default=False}
]
