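# Physics asset for the M_MED_ButtonCase male cosmetic body, rigging its skirt-like garment bones for secondary motion with pelvis, spine and leg capsules.
# Eight four-link chains (front, middle and two back strands per side) hang from deform anchors on spine_01, each joint limited on swing1 with other angular axes locked.
# Cross constraints tie neighbouring strands around the back and sides, leaving the front open.
# A huge root sphere sits below the origin, apparently as a floor collider.
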PhysicsAsset /BRCosmetics/Characters/Player/Male/Medium/Bodies/M_MED_ButtonCase/Meshes/M_MED_ButtonCase_Physics
bodies {
 calf_l [{shapeType=Capsule, length=63.5998, radius=8.6844, shapeName=calf_r_capsule, center=<-29.6236 0 0>, rotation=<rot 90 180 -177.4687>}]
 calf_r [{shapeType=Capsule, length=63.5998, radius=8.6844, shapeName=calf_l_capsule, center=<29.6236 0 0>, rotation=<rot 90 180 2.5313>}]
 deform_skirt_b_l [{shapeType=Capsule, length=1, radius=1, shapeName=deform_skirt_b_l_capsule, center=<12.539 -3.7658 0>, rotation=<rot 90 0 0>}]
 deform_skirt_b_r [{shapeType=Capsule, length=1, radius=1, shapeName=deform_skirt_b_r_capsule, center=<12.539 -3.7658 0>, rotation=<rot 90 0 0>}]
 deform_skirt_bk_l [{shapeType=Capsule, length=1, radius=1, shapeName=deform_skirt_bk_l_capsule, center=<12.539 -3.7658 0>, rotation=<rot 90 0 0>}]
 deform_skirt_bk_r [{shapeType=Capsule, length=1, radius=1, shapeName=deform_skirt_bk_r_capsule, center=<12.539 -3.7658 0>, rotation=<rot 90 0 0>}]
 deform_skirt_f_l [{shapeType=Capsule, length=1, radius=1, shapeName=deform_skirt_f_l_capsule, rotation=<rot 90 0 0>}]
 deform_skirt_f_r [{shapeType=Capsule, length=1, radius=1, shapeName=deform_skirt_f_r_capsule, center=<12.539 -3.7658 0>, rotation=<rot 90 0 0>}]
 deform_skirt_m_l [{shapeType=Capsule, length=1, radius=1, shapeName=deform_skirt_m_l_capsule, center=<12.539 -3.7658 0>, rotation=<rot 90 0 0>}]
 deform_skirt_m_r [{shapeType=Capsule, length=1, radius=1, shapeName=deform_skirt_m_r_capsule, center=<12.539 -3.7658 0>, rotation=<rot 90 0 0>}]
 dyn_skirt_b_1_l [{shapeType=Capsule, length=21.6479, radius=10.2465, shapeName=dyn_skirt_b_1_l_capsule, center=<12.539 -8.3875 0>}]
 dyn_skirt_b_1_r [{shapeType=Capsule, length=21.6479, radius=10.2465, shapeName=dyn_skirt_b_1_r_capsule, center=<12.539 -8.3875 0>}]
 dyn_skirt_b_2_l [{shapeType=Capsule, length=26.5351, radius=12.1771, shapeName=dyn_skirt_b_2_l_capsule, center=<12.539 -6.5686 0>}]
 dyn_skirt_b_2_r [{shapeType=Capsule, length=26.5351, radius=12.1771, shapeName=dyn_skirt_b_2_r_capsule, center=<12.539 -6.5686 0>}]
 dyn_skirt_b_3_l [{shapeType=Capsule, length=25.4366, radius=13.2756, shapeName=dyn_skirt_b_3_l_capsule, center=<12.539 -7.1206 0>}]
 dyn_skirt_b_3_r [{shapeType=Capsule, length=25.4366, radius=13.2756, shapeName=dyn_skirt_b_3_r_capsule, center=<12.539 -7.1206 0>}]
 dyn_skirt_b_l [{shapeType=Capsule, length=21.6479, radius=10.2465, shapeName=dyn_skirt_b_l_capsule, center=<12.539 -10.0657 0>}]
 dyn_skirt_b_r [{shapeType=Capsule, length=21.6479, radius=10.2465, shapeName=dyn_skirt_b_r_capsule, center=<12.539 -10.0657 0>}]
 dyn_skirt_bk_1_l [{shapeType=Capsule, length=21.6479, radius=10.2465, shapeName=dyn_skirt_bk_1_l_capsule, center=<12.539 -8.3875 0>}]
 dyn_skirt_bk_1_r [{shapeType=Capsule, length=21.6479, radius=10.2465, shapeName=dyn_skirt_bk_1_r_capsule, center=<12.539 -8.3875 0>}]
 dyn_skirt_bk_2_l [{shapeType=Capsule, length=26.5351, radius=12.1771, shapeName=dyn_skirt_bk_2_l_capsule, center=<13.9087 -6.5686 0>, rotation=<rot 0 0 -180>}]
 dyn_skirt_bk_2_r [{shapeType=Capsule, length=26.5351, radius=12.1771, shapeName=dyn_skirt_bk_2_r_capsule, center=<12.539 -6.5686 0>}]
 dyn_skirt_bk_3_l [{shapeType=Capsule, length=25.4366, radius=13.2756, shapeName=dyn_skirt_bk_3_l_capsule, center=<12.539 -7.1206 0>}]
 dyn_skirt_bk_3_r [{shapeType=Capsule, length=25.4366, radius=13.2756, shapeName=dyn_skirt_bk_3_r_capsule, center=<12.539 -7.1206 0>}]
 dyn_skirt_bk_l [{shapeType=Capsule, length=21.6479, radius=10.2465, shapeName=dyn_skirt_bk_l_capsule, center=<12.539 -10.0657 0>}]
 dyn_skirt_bk_r [{shapeType=Capsule, length=21.6479, radius=10.2465, shapeName=dyn_skirt_bk_r_capsule, center=<12.539 -10.0657 0>}]
 dyn_skirt_f_1_l [{shapeType=Capsule, length=18.3235, radius=7, shapeName=dyn_skirt_f_1_l_capsule, center=<9.0566 -6.0872 -1.8564>, rotation=<rot 90 0 0>}]
 dyn_skirt_f_1_r [{shapeType=Capsule, length=15, radius=7, shapeName=dyn_skirt_f_1_r_capsule, center=<8.8226 -6.0908 0>, rotation=<rot 90 0 0>}]
 dyn_skirt_f_2_l [{shapeType=Capsule, length=19.8454, radius=7, shapeName=dyn_skirt_f_2_l_capsule, center=<10.1157 -6.5494 0>, rotation=<rot 90 0 0>}]
 dyn_skirt_f_2_r [{
   shapeType=Capsule
   length=18.4606
   radius=8.3622
   shapeName=dyn_skirt_f_2_r_capsule
   center=<10.812 -5.6561 2.1628>
   rotation=<rot 90 180 176.4548>
  }]
 dyn_skirt_f_3_l [{shapeType=Capsule, length=15, radius=7, shapeName=dyn_skirt_f_3_l_capsule, center=<9.5244 -5.3463 0>, rotation=<rot 90 0 0>}]
 dyn_skirt_f_3_r [{shapeType=Capsule, length=15.7114, radius=6.2886, shapeName=dyn_skirt_f_3_r_capsule, center=<8.8226 -3.8671 0>, rotation=<rot 90 0 0>}]
 dyn_skirt_f_l [{
   shapeType=Capsule
   length=12.6592
   radius=7
   shapeName=dyn_skirt_f_l_capsule
   center=<6.4659 -6.6196 -1.9266>
   rotation=<rot 90 -179.7762 179.8072>
  }]
 dyn_skirt_f_r [{shapeType=Capsule, length=15, radius=7, shapeName=dyn_skirt_f_r_capsule, center=<8.8226 -6.0908 0>, rotation=<rot 90 0 0>}]
 dyn_skirt_m_1_l [{shapeType=Capsule, length=11.9441, radius=10.2465, shapeName=dyn_skirt_m_1_l_capsule, center=<12.539 -8.6419 0>}]
 dyn_skirt_m_1_r [{shapeType=Capsule, length=11.7535, radius=10.2465, shapeName=dyn_skirt_m_1_r_capsule, center=<12.539 -8.7539 0>}]
 dyn_skirt_m_2_l [{shapeType=Capsule, length=13.3699, radius=10.2465, shapeName=dyn_skirt_m_2_l_capsule, center=<12.539 -8.7539 0>}]
 dyn_skirt_m_2_r [{shapeType=Capsule, length=22.0701, radius=10.2465, shapeName=dyn_skirt_m_2_r_capsule, center=<12.539 -8.7539 0>}]
 dyn_skirt_m_3_l [{shapeType=Capsule, length=17.8882, radius=10.2465, shapeName=dyn_skirt_m_3_l_capsule, center=<12.539 -8.7539 0>}]
 dyn_skirt_m_3_r [{shapeType=Capsule, length=11.6623, radius=12.9829, shapeName=dyn_skirt_m_3_r_capsule, center=<12.539 -10.2698 0>}]
 dyn_skirt_m_l [{shapeType=Capsule, length=21.6194, radius=10.2465, shapeName=dyn_skirt_m_l_capsule, center=<12.539 -8.7539 0>, rotation=<rot 29.045 0 0>}]
 dyn_skirt_m_r [{shapeType=Capsule, length=18.4855, radius=10.2465, shapeName=dyn_skirt_m_r_capsule, center=<12.539 -8.7539 0>, rotation=<rot -26.5196 0 0>}]
 pelvis [{shapeType=Capsule, length=15, radius=7, shapeName=pelvis_capsule, rotation=<rot 90 0 0>}]
 root [{shapeType=Sphere, radius=668.1761, shapeName=root_sphere, center=<0 23.0851 -666.2379>}]
 spine_01 [{shapeType=Capsule, length=15, radius=7, shapeName=spine_01_capsule, rotation=<rot 90 0 0>}]
 thigh_l [{shapeType=Capsule, length=27.8137, radius=8, shapeName=thigh_r_capsule, center=<-20.5159 0 0>, rotation=<rot 90 0 180>}]
 thigh_r [{shapeType=Capsule, length=27.8137, radius=8, shapeName=thigh_l_capsule, center=<20.5159 0 0>, rotation=<rot 90 0 0>}]
}
constraints [
 {
  bone1Name=pelvis
  bone2Name=root
  swing1LimitDegrees=45
  swing1Motion=Free
  swing2LimitDegrees=45
  swing2Motion=Free
  twistLimitDegrees=45
  twistMotion=Free
 }
 {
  bone1Name=spine_01
  bone2Name=pelvis
  swing1LimitDegrees=45
  swing1Motion=Free
  swing2LimitDegrees=45
  swing2Motion=Free
  twistLimitDegrees=45
  twistMotion=Free
 }
 {
  bone1Name=deform_skirt_b_l
  bone2Name=spine_01
  swing1LimitDegrees=45
  swing1Motion=Free
  swing2LimitDegrees=45
  swing2Motion=Free
  twistLimitDegrees=45
  twistMotion=Free
 }
 {
  bone1Name=dyn_skirt_b_l
  bone2Name=deform_skirt_b_l
  swing1LimitDegrees=25
  swing2LimitDegrees=1
  swing2Motion=Locked
  twistLimitDegrees=1
  twistMotion=Locked
 }
 {
  bone1Name=dyn_skirt_b_1_l
  bone2Name=dyn_skirt_b_l
  swing1LimitDegrees=30
  swing2LimitDegrees=1
  swing2Motion=Locked
  twistLimitDegrees=1
  twistMotion=Locked
 }
 {
  bone1Name=dyn_skirt_b_2_l
  bone2Name=dyn_skirt_b_1_l
  swing1LimitDegrees=35
  swing2LimitDegrees=1
  swing2Motion=Locked
  twistLimitDegrees=1
  twistMotion=Locked
 }
 {
  bone1Name=dyn_skirt_b_3_l
  bone2Name=dyn_skirt_b_2_l
  swing1LimitDegrees=10
  swing2LimitDegrees=2
  swing2Motion=Locked
  twistLimitDegrees=1
  twistMotion=Locked
 }
 {
  bone1Name=deform_skirt_b_r
  bone2Name=spine_01
  swing1LimitDegrees=45
  swing1Motion=Free
  swing2LimitDegrees=45
  swing2Motion=Free
  twistLimitDegrees=45
  twistMotion=Free
 }
 {
  bone1Name=dyn_skirt_b_r
  bone2Name=deform_skirt_b_r
  swing1LimitDegrees=8
  swing2LimitDegrees=1
  swing2Motion=Locked
  twistLimitDegrees=1
  twistMotion=Locked
 }
 {
  bone1Name=dyn_skirt_b_1_r
  bone2Name=dyn_skirt_b_r
  swing1LimitDegrees=30
  swing2LimitDegrees=1
  swing2Motion=Locked
  twistLimitDegrees=1
  twistMotion=Locked
 }
 {
  bone1Name=dyn_skirt_b_2_r
  bone2Name=dyn_skirt_b_1_r
  swing1LimitDegrees=25
  swing2LimitDegrees=1
  swing2Motion=Locked
  twistLimitDegrees=1
  twistMotion=Locked
 }
 {
  bone1Name=dyn_skirt_b_3_r
  bone2Name=dyn_skirt_b_2_r
  swing1LimitDegrees=10
  swing2LimitDegrees=2
  swing2Motion=Locked
  twistLimitDegrees=1
  twistMotion=Locked
 }
 {
  bone1Name=deform_skirt_bk_l
  bone2Name=spine_01
  swing1LimitDegrees=45
  swing1Motion=Free
  swing2LimitDegrees=45
  swing2Motion=Free
  twistLimitDegrees=45
  twistMotion=Free
 }
 {
  bone1Name=dyn_skirt_bk_l
  bone2Name=deform_skirt_bk_l
  swing1LimitDegrees=15
  swing2LimitDegrees=1
  swing2Motion=Locked
  twistLimitDegrees=1
  twistMotion=Locked
 }
 {
  bone1Name=dyn_skirt_bk_1_l
  bone2Name=dyn_skirt_bk_l
  swing1LimitDegrees=30
  swing2LimitDegrees=1
  swing2Motion=Locked
  twistLimitDegrees=1
  twistMotion=Locked
 }
 {
  bone1Name=dyn_skirt_bk_2_l
  bone2Name=dyn_skirt_bk_1_l
  swing1LimitDegrees=15
  swing2LimitDegrees=1
  swing2Motion=Locked
  twistLimitDegrees=1
  twistMotion=Locked
 }
 {
  bone1Name=dyn_skirt_bk_3_l
  bone2Name=dyn_skirt_bk_2_l
  swing1LimitDegrees=10
  swing2LimitDegrees=2
  swing2Motion=Locked
  twistLimitDegrees=1
  twistMotion=Locked
 }
 {
  bone1Name=deform_skirt_bk_r
  bone2Name=spine_01
  swing1LimitDegrees=45
  swing1Motion=Free
  swing2LimitDegrees=45
  swing2Motion=Free
  twistLimitDegrees=45
  twistMotion=Free
 }
 {
  bone1Name=dyn_skirt_bk_r
  bone2Name=deform_skirt_bk_r
  swing1LimitDegrees=15
  swing2LimitDegrees=1
  swing2Motion=Locked
  twistLimitDegrees=1
  twistMotion=Locked
 }
 {
  bone1Name=dyn_skirt_bk_1_r
  bone2Name=dyn_skirt_bk_r
  swing1LimitDegrees=30
  swing2LimitDegrees=1
  swing2Motion=Locked
  twistLimitDegrees=1
  twistMotion=Locked
 }
 {
  bone1Name=dyn_skirt_bk_2_r
  bone2Name=dyn_skirt_bk_1_r
  swing1LimitDegrees=15
  swing2LimitDegrees=1
  swing2Motion=Locked
  twistLimitDegrees=1
  twistMotion=Locked
 }
 {
  bone1Name=dyn_skirt_bk_3_r
  bone2Name=dyn_skirt_bk_2_r
  swing1LimitDegrees=10
  swing2LimitDegrees=2
  swing2Motion=Locked
  twistLimitDegrees=1
  twistMotion=Locked
 }
 {
  bone1Name=deform_skirt_f_l
  bone2Name=spine_01
  swing1LimitDegrees=45
  swing1Motion=Free
  swing2LimitDegrees=45
  swing2Motion=Free
  twistLimitDegrees=45
  twistMotion=Free
 }
 {
  bone1Name=dyn_skirt_f_l
  bone2Name=deform_skirt_f_l
  swing1LimitDegrees=10
  swing2LimitDegrees=1
  swing2Motion=Locked
  twistLimitDegrees=1
  twistMotion=Locked
 }
 {
  bone1Name=dyn_skirt_f_1_l
  bone2Name=dyn_skirt_f_l
  swing1LimitDegrees=35
  swing2LimitDegrees=1
  swing2Motion=Locked
  twistLimitDegrees=1
  twistMotion=Locked
 }
 {
  bone1Name=dyn_skirt_f_2_l
  bone2Name=dyn_skirt_f_1_l
  swing1LimitDegrees=35
  swing2LimitDegrees=1
  swing2Motion=Locked
  twistLimitDegrees=1
  twistMotion=Locked
 }
 {
  bone1Name=dyn_skirt_f_3_l
  bone2Name=dyn_skirt_f_2_l
  swing1LimitDegrees=35
  swing2LimitDegrees=2
  swing2Motion=Locked
  twistLimitDegrees=1
  twistMotion=Locked
 }
 {
  bone1Name=deform_skirt_f_r
  bone2Name=spine_01
  swing1LimitDegrees=45
  swing1Motion=Free
  swing2LimitDegrees=45
  swing2Motion=Free
  twistLimitDegrees=45
  twistMotion=Free
 }
 {
  bone1Name=dyn_skirt_f_r
  bone2Name=deform_skirt_f_r
  swing1LimitDegrees=8
  swing2LimitDegrees=1
  swing2Motion=Locked
  twistLimitDegrees=1
  twistMotion=Locked
 }
 {
  bone1Name=dyn_skirt_f_1_r
  bone2Name=dyn_skirt_f_r
  swing1LimitDegrees=45
  swing2LimitDegrees=1
  swing2Motion=Locked
  twistLimitDegrees=1
  twistMotion=Locked
 }
 {
  bone1Name=dyn_skirt_f_2_r
  bone2Name=dyn_skirt_f_1_r
  swing1LimitDegrees=35
  swing2LimitDegrees=1
  swing2Motion=Locked
  twistLimitDegrees=1
  twistMotion=Locked
 }
 {
  bone1Name=dyn_skirt_f_3_r
  bone2Name=dyn_skirt_f_2_r
  swing1LimitDegrees=30
  swing2LimitDegrees=2
  swing2Motion=Locked
  twistLimitDegrees=1
  twistMotion=Locked
 }
 {
  bone1Name=deform_skirt_m_l
  bone2Name=spine_01
  swing1LimitDegrees=45
  swing1Motion=Free
  swing2LimitDegrees=45
  swing2Motion=Free
  twistLimitDegrees=45
  twistMotion=Free
 }
 {
  bone1Name=dyn_skirt_m_l
  bone2Name=deform_skirt_m_l
  swing1LimitDegrees=8
  swing2LimitDegrees=1
  swing2Motion=Locked
  twistLimitDegrees=1
  twistMotion=Locked
 }
 {
  bone1Name=dyn_skirt_m_1_l
  bone2Name=dyn_skirt_m_l
  swing1LimitDegrees=45
  swing2LimitDegrees=1
  swing2Motion=Locked
  twistLimitDegrees=1
  twistMotion=Locked
 }
 {
  bone1Name=dyn_skirt_m_2_l
  bone2Name=dyn_skirt_m_1_l
  swing1LimitDegrees=35
  swing2LimitDegrees=1
  swing2Motion=Locked
  twistLimitDegrees=1
  twistMotion=Locked
 }
 {
  bone1Name=dyn_skirt_m_3_l
  bone2Name=dyn_skirt_m_2_l
  swing1LimitDegrees=35
  swing2LimitDegrees=2
  swing2Motion=Locked
  twistLimitDegrees=1
  twistMotion=Locked
 }
 {
  bone1Name=deform_skirt_m_r
  bone2Name=spine_01
  swing1LimitDegrees=45
  swing1Motion=Free
  swing2LimitDegrees=45
  swing2Motion=Free
  twistLimitDegrees=45
  twistMotion=Free
 }
 {
  bone1Name=dyn_skirt_m_r
  bone2Name=deform_skirt_m_r
  swing1LimitDegrees=8
  swing2LimitDegrees=1
  swing2Motion=Locked
  twistLimitDegrees=1
  twistMotion=Locked
 }
 {
  bone1Name=dyn_skirt_m_1_r
  bone2Name=dyn_skirt_m_r
  swing1LimitDegrees=45
  swing2LimitDegrees=1
  swing2Motion=Locked
  twistLimitDegrees=1
  twistMotion=Locked
 }
 {
  bone1Name=dyn_skirt_m_2_r
  bone2Name=dyn_skirt_m_1_r
  swing1LimitDegrees=35
  swing2LimitDegrees=1
  swing2Motion=Locked
  twistLimitDegrees=1
  twistMotion=Locked
 }
 {
  bone1Name=dyn_skirt_m_3_r
  bone2Name=dyn_skirt_m_2_r
  swing1LimitDegrees=10
  swing2LimitDegrees=2
  swing2Motion=Locked
  twistLimitDegrees=1
  twistMotion=Locked
 }
 {
  bone1Name=thigh_r
  bone2Name=pelvis
  swing1LimitDegrees=45
  swing1Motion=Free
  swing2LimitDegrees=45
  swing2Motion=Free
  twistLimitDegrees=45
  twistMotion=Free
 }
 {
  bone1Name=calf_r
  bone2Name=thigh_r
  swing1LimitDegrees=45
  swing1Motion=Free
  swing2LimitDegrees=45
  swing2Motion=Free
  twistLimitDegrees=45
  twistMotion=Free
 }
 {
  bone1Name=thigh_l
  bone2Name=pelvis
  swing1LimitDegrees=45
  swing1Motion=Free
  swing2LimitDegrees=45
  swing2Motion=Free
  twistLimitDegrees=45
  twistMotion=Free
 }
 {
  bone1Name=calf_l
  bone2Name=thigh_l
  swing1LimitDegrees=45
  swing1Motion=Free
  swing2LimitDegrees=45
  swing2Motion=Free
  twistLimitDegrees=45
  twistMotion=Free
 }
 {
  bone1Name=dyn_skirt_bk_l
  bone2Name=dyn_skirt_bk_r
  swing1LimitDegrees=20
  swing2LimitDegrees=45
  swing2Motion=Free
  twistLimitDegrees=45
  twistMotion=Free
 }
 {
  bone1Name=dyn_skirt_bk_1_l
  bone2Name=dyn_skirt_bk_1_r
  swing1LimitDegrees=20
  swing2LimitDegrees=45
  swing2Motion=Free
  twistLimitDegrees=45
  twistMotion=Free
 }
 {
  bone1Name=dyn_skirt_bk_2_l
  bone2Name=dyn_skirt_bk_2_r
  swing1LimitDegrees=20
  swing2LimitDegrees=45
  swing2Motion=Free
  twistLimitDegrees=45
  twistMotion=Free
 }
 {
  bone1Name=dyn_skirt_bk_3_l
  bone2Name=dyn_skirt_bk_3_r
  swing1LimitDegrees=20
  swing2LimitDegrees=45
  swing2Motion=Free
  twistLimitDegrees=45
  twistMotion=Free
 }
 {
  bone1Name=dyn_skirt_b_l
  bone2Name=dyn_skirt_bk_l
  swing1LimitDegrees=20
  swing2LimitDegrees=45
  swing2Motion=Free
  twistLimitDegrees=45
  twistMotion=Free
 }
 {
  bone1Name=dyn_skirt_b_1_l
  bone2Name=dyn_skirt_bk_1_l
  swing1LimitDegrees=35
  swing2LimitDegrees=45
  swing2Motion=Free
  twistLimitDegrees=45
  twistMotion=Free
 }
 {
  bone1Name=dyn_skirt_b_2_l
  bone2Name=dyn_skirt_bk_2_l
  swing1LimitDegrees=35
  swing2LimitDegrees=45
  swing2Motion=Free
  twistLimitDegrees=45
  twistMotion=Free
 }
 {
  bone1Name=dyn_skirt_b_3_l
  bone2Name=dyn_skirt_bk_3_l
  swing1LimitDegrees=35
  swing2LimitDegrees=45
  swing2Motion=Free
  twistLimitDegrees=45
  twistMotion=Free
 }
 {
  bone1Name=dyn_skirt_b_r
  bone2Name=dyn_skirt_bk_r
  swing1LimitDegrees=20
  swing2LimitDegrees=45
  swing2Motion=Free
  twistLimitDegrees=45
  twistMotion=Free
 }
 {
  bone1Name=dyn_skirt_b_1_r
  bone2Name=dyn_skirt_bk_1_r
  swing1LimitDegrees=35
  swing2LimitDegrees=45
  swing2Motion=Free
  twistLimitDegrees=45
  twistMotion=Free
 }
 {
  bone1Name=dyn_skirt_b_2_r
  bone2Name=dyn_skirt_bk_2_r
  swing1LimitDegrees=35
  swing2LimitDegrees=45
  swing2Motion=Free
  twistLimitDegrees=45
  twistMotion=Free
 }
 {
  bone1Name=dyn_skirt_b_3_r
  bone2Name=dyn_skirt_bk_3_r
  swing1LimitDegrees=35
  swing2LimitDegrees=45
  swing2Motion=Free
  twistLimitDegrees=45
  twistMotion=Free
 }
 {
  bone1Name=dyn_skirt_m_l
  bone2Name=dyn_skirt_b_l
  swing1LimitDegrees=20
  swing2LimitDegrees=45
  swing2Motion=Free
  twistLimitDegrees=10
  twistMotion=Free
 }
 {
  bone1Name=dyn_skirt_m_1_l
  bone2Name=dyn_skirt_b_1_l
  swing1LimitDegrees=35
  swing2LimitDegrees=45
  swing2Motion=Free
  twistLimitDegrees=45
  twistMotion=Free
 }
 {
  bone1Name=dyn_skirt_m_2_l
  bone2Name=dyn_skirt_b_2_l
  swing1LimitDegrees=35
  swing2LimitDegrees=45
  swing2Motion=Free
  twistLimitDegrees=45
  twistMotion=Free
 }
 {
  bone1Name=dyn_skirt_f_l
  bone2Name=dyn_skirt_m_l
  swing1LimitDegrees=10
  swing2LimitDegrees=45
  swing2Motion=Free
  twistLimitDegrees=10
  twistMotion=Free
 }
 {
  bone1Name=dyn_skirt_f_1_l
  bone2Name=dyn_skirt_m_1_l
  swing1LimitDegrees=10
  swing2LimitDegrees=45
  swing2Motion=Free
  twistLimitDegrees=45
  twistMotion=Free
 }
 {
  bone1Name=dyn_skirt_f_2_l
  bone2Name=dyn_skirt_m_2_l
  swing1LimitDegrees=10
  swing2LimitDegrees=45
  swing2Motion=Free
  twistLimitDegrees=45
  twistMotion=Free
 }
 {
  bone1Name=dyn_skirt_f_3_l
  bone2Name=dyn_skirt_m_3_l
  swing1LimitDegrees=10
  swing2LimitDegrees=45
  swing2Motion=Free
  twistLimitDegrees=45
  twistMotion=Free
 }
 {
  bone1Name=dyn_skirt_m_r
  bone2Name=dyn_skirt_b_r
  swing1LimitDegrees=20
  swing2LimitDegrees=45
  swing2Motion=Free
  twistLimitDegrees=45
  twistMotion=Free
 }
 {
  bone1Name=dyn_skirt_m_1_r
  bone2Name=dyn_skirt_b_1_r
  swing1LimitDegrees=35
  swing2LimitDegrees=45
  swing2Motion=Free
  twistLimitDegrees=45
  twistMotion=Free
 }
 {
  bone1Name=dyn_skirt_m_2_r
  bone2Name=dyn_skirt_b_2_r
  swing1LimitDegrees=35
  swing2LimitDegrees=45
  swing2Motion=Free
  twistLimitDegrees=45
  twistMotion=Free
 }
 {
  bone1Name=dyn_skirt_f_r
  bone2Name=dyn_skirt_m_r
  swing1LimitDegrees=10
  swing2LimitDegrees=45
  swing2Motion=Free
  twistLimitDegrees=45
  twistMotion=Free
 }
 {
  bone1Name=dyn_skirt_f_1_r
  bone2Name=dyn_skirt_m_1_r
  swing1LimitDegrees=10
  swing2LimitDegrees=45
  swing2Motion=Free
  twistLimitDegrees=45
  twistMotion=Free
 }
 {
  bone1Name=dyn_skirt_f_2_r
  bone2Name=dyn_skirt_m_2_r
  swing1LimitDegrees=10
  swing2LimitDegrees=45
  swing2Motion=Free
  twistLimitDegrees=45
  twistMotion=Free
 }
 {
  bone1Name=dyn_skirt_f_3_r
  bone2Name=dyn_skirt_m_3_r
  swing1LimitDegrees=10
  swing2LimitDegrees=45
  swing2Motion=Free
  twistLimitDegrees=45
  twistMotion=Free
 }
]
