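# Physics asset for the upper-body mesh of the male medium ButtonCase player body, rigging its collar and chain secondary motion.
# Capsules cover spine, neck, head, clavicles, pecs and only the left upper arm, joined by free-rotation constraints.
# Collar and upper/lower chain dyn bones hang from deform anchors with limited or locked joints.
# Extra cross-constraints tie inner and outer chain ends together on each side, forming closed loops.
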PhysicsAsset /BRCosmetics/Characters/Player/Male/Medium/Bodies/M_MED_ButtonCase/Meshes/M_MED_ButtonCase_Upper_Physics
bodies {
 clavicle_l [{shapeType=Capsule, length=9.2639, radius=8.558, shapeName=clavicle_r_capsule, center=<5.9515 0 0>, rotation=<rot 90 0 180>}]
 clavicle_r [{shapeType=Capsule, length=5.7035, radius=7.889, shapeName=clavicle_r_capsule, center=<-7.6694 0 0>, rotation=<rot 90 0 0>}]
 deform_chain_down_inn_l [{shapeType=Capsule, length=1, radius=1, shapeName=deform_chain_down_inn_l_capsule}]
 deform_chain_down_inn_r [{shapeType=Capsule, length=1, radius=1, shapeName=deform_chain_down_inn_r_capsule}]
 deform_chain_up_inn_l [{shapeType=Capsule, length=1, radius=1, shapeName=deform_chain_up_inn_l_capsule}]
 deform_chain_up_inn_r [{shapeType=Capsule, length=1, radius=1, shapeName=deform_chain_up_inn_r_capsule}]
 deform_chain_up_out_l [{shapeType=Capsule, length=1, radius=1, shapeName=deform_chain_up_out_l_capsule}]
 deform_chain_up_out_r [{shapeType=Capsule, length=1, radius=1, shapeName=deform_chain_up_out_r_capsule}]
 deform_collar_l [{shapeType=Capsule, length=1, radius=1, shapeName=deform_collar_l_capsule}]
 deform_collar_r [{shapeType=Capsule, length=1, radius=1, shapeName=deform_collar_r_capsule}]
 deform_pec_aim_base_l [{shapeType=Capsule, length=1, radius=1, shapeName=deform_pec_aim_base_r_capsule, center=<8.803 -3.0435 0>, rotation=<rot 90 0 180>}]
 deform_pec_aim_base_r [{shapeType=Capsule, length=1, radius=1, shapeName=deform_pec_aim_base_r_capsule, center=<-8.803 3.0435 0>, rotation=<rot 90 0 0>}]
 deform_pec_aim_l [{
   shapeType=Capsule
   length=7.4003
   radius=9.4811
   shapeName=deform_pec_aim_r_capsule
   center=<9.1786 -4.9439 -1.7326>
   rotation=<rot 90 0 168.819>
  }]
 deform_pec_aim_r [{
   shapeType=Capsule
   length=13.7009
   radius=7.0937
   shapeName=deform_pec_aim_r_capsule
   center=<-11.1417 2.0038 0>
   rotation=<rot 90 180 166.4715>
  }]
 dyn_chain_down_inn_1_l [{shapeType=Capsule, length=5, radius=2, shapeName=dyn_chain_down_inn_1_l_capsule, center=<3.9752 -0.9733 0>, rotation=<rot 90 0 0>}]
 dyn_chain_down_inn_1_r [{shapeType=Capsule, length=5, radius=2, shapeName=dyn_chain_down_inn_1_r_capsule, center=<3.9752 -0.9733 0>, rotation=<rot 90 0 0>}]
 dyn_chain_down_inn_2_l [{shapeType=Capsule, length=5, radius=2, shapeName=dyn_chain_down_inn_2_l_capsule, center=<0 -8.6596 0>, rotation=<rot 90 0 0>}]
 dyn_chain_down_inn_2_r [{shapeType=Capsule, length=5, radius=2, shapeName=dyn_chain_down_inn_2_r_capsule, center=<0 -8.6596 0>, rotation=<rot 90 0 0>}]
 dyn_chain_down_inn_l [{shapeType=Capsule, length=9.6625, radius=2, shapeName=dyn_chain_down_inn_l_capsule, center=<5.7227 -1.4573 0>, rotation=<rot 90 0 0>}]
 dyn_chain_down_inn_r [{shapeType=Capsule, length=9.6625, radius=2, shapeName=dyn_chain_down_inn_r_capsule, center=<5.7227 -1.4573 0>, rotation=<rot 90 0 0>}]
 dyn_chain_down_out_1_l [{shapeType=Capsule, length=5, radius=2, shapeName=dyn_chain_down_out_1_l_capsule, center=<3.9752 -0.9733 0>, rotation=<rot 90 0 0>}]
 dyn_chain_down_out_1_r [{
   shapeType=Capsule
   length=7.1604
   radius=2
   shapeName=dyn_chain_down_out_1_r_capsule
   center=<3.3086 0 -1.0123>
   rotation=<rot 72.9884 -0.0001 0>
  }]
 dyn_chain_down_out_2_l [{shapeType=Capsule, length=5, radius=2, shapeName=dyn_chain_down_out_2_l_capsule, center=<0 -1.2624 0>, rotation=<rot 90 0 0>}]
 dyn_chain_down_out_2_r [{shapeType=Capsule, length=5.4384, radius=2, shapeName=dyn_chain_down_out_2_r_capsule, center=<0 -1.0964 0>, rotation=<rot 90 0 0>}]
 dyn_chain_up_inn_1_l [{shapeType=Capsule, length=5, radius=2, shapeName=dyn_chain_up_inn_1_l_capsule, center=<0 -1.4451 0>, rotation=<rot 90 0 0>}]
 dyn_chain_up_inn_1_r [{shapeType=Capsule, length=5, radius=2, shapeName=dyn_chain_up_inn_1_r_capsule, center=<0 -1.4451 0>, rotation=<rot 90 0 0>}]
 dyn_chain_up_inn_l [{shapeType=Capsule, length=5, radius=2, shapeName=dyn_chain_up_inn_l_capsule, center=<4.1618 -2.0795 0>, rotation=<rot 90 0 0>}]
 dyn_chain_up_inn_r [{shapeType=Capsule, length=5, radius=2, shapeName=dyn_chain_up_inn_r_capsule, center=<4.1618 -2.0795 0>, rotation=<rot 90 0 0>}]
 dyn_chain_up_out_1_l [{shapeType=Capsule, length=5, radius=2, shapeName=dyn_chain_up_out_1_l_capsule, center=<0 -1.4451 0>, rotation=<rot 90 0 0>}]
 dyn_chain_up_out_1_r [{shapeType=Capsule, length=5, radius=2, shapeName=dyn_chain_up_out_1_r_capsule, center=<0 -1.0964 0>, rotation=<rot 90 0 0>}]
 dyn_chain_up_out_l [{shapeType=Capsule, length=5, radius=2, shapeName=dyn_chain_up_out_l_capsule, center=<3.9752 -0.9733 0>, rotation=<rot 90 0 0>}]
 dyn_chain_up_out_r [{shapeType=Capsule, length=5, radius=2, shapeName=dyn_chain_up_out_r_capsule, center=<3.9752 -0.9733 0>, rotation=<rot 90 0 180>}]
 dyn_collar_l [{shapeType=Capsule, length=5, radius=3, shapeName=dyn_collar_l_capsule, center=<4.4313 -1.5225 0>}]
 dyn_collar_r [{shapeType=Capsule, length=5, radius=3, shapeName=dyn_collar_r_capsule, center=<4.4313 -1.5225 0>}]
 head [{shapeType=Capsule, length=6.7802, radius=5.9017, shapeName=head_capsule, center=<1.2375 3.5868 0>, rotation=<rot 90 -26.5651 22.0299>}]
 neck_02 [{shapeType=Capsule, length=4.9099, radius=5.0901, shapeName=neck_02_capsule, center=<1.4672 0 0>, rotation=<rot 90 0 0>}]
 spine_04 [{shapeType=Capsule, length=2, radius=15.4746, shapeName=spine_04_capsule, center=<2.3564 0 0>}]
 spine_05 [{shapeType=Capsule, length=11.6425, radius=12.9021, shapeName=spine_05_capsule}]
 upperarm_l [{
   shapeType=Capsule
   length=20.1964
   radius=5.8955
   shapeName=upperarm_l_capsule
   center=<9.7606 0.6128 -0.9719>
   rotation=<rot 90 180 176.2338>
  }]
}
constraints [
 {
  bone1Name=upperarm_l
  bone2Name=clavicle_l
  swing1LimitDegrees=45
  swing1Motion=Free
  swing2LimitDegrees=45
  swing2Motion=Free
  twistLimitDegrees=45
  twistMotion=Free
 }
 {
  bone1Name=clavicle_r
  bone2Name=spine_05
  swing1LimitDegrees=45
  swing1Motion=Free
  swing2LimitDegrees=45
  swing2Motion=Free
  twistLimitDegrees=45
  twistMotion=Free
 }
 {
  bone1Name=deform_chain_down_inn_l
  bone2Name=spine_05
  swing1LimitDegrees=45
  swing1Motion=Free
  swing2LimitDegrees=45
  swing2Motion=Free
  twistLimitDegrees=45
  twistMotion=Free
 }
 {bone1Name=dyn_chain_down_inn_l, bone2Name=deform_chain_down_inn_l, swing1LimitDegrees=35, swing2LimitDegrees=15, twistLimitDegrees=1}
 {bone1Name=dyn_chain_down_inn_1_l, bone2Name=dyn_chain_down_inn_l, swing1LimitDegrees=35, swing2LimitDegrees=15, twistLimitDegrees=1}
 {
  bone1Name=dyn_chain_down_inn_2_l
  bone2Name=dyn_chain_down_inn_1_l
  swing1LimitDegrees=5
  swing1Motion=Locked
  swing2LimitDegrees=5
  swing2Motion=Locked
  twistLimitDegrees=1
  twistMotion=Locked
 }
 {
  bone1Name=deform_chain_down_inn_r
  bone2Name=spine_05
  swing1LimitDegrees=45
  swing1Motion=Free
  swing2LimitDegrees=45
  swing2Motion=Free
  twistLimitDegrees=45
  twistMotion=Free
 }
 {bone1Name=dyn_chain_down_inn_r, bone2Name=deform_chain_down_inn_r, swing1LimitDegrees=35, swing2LimitDegrees=15, twistLimitDegrees=1}
 {bone1Name=dyn_chain_down_inn_1_r, bone2Name=dyn_chain_down_inn_r, swing1LimitDegrees=35, swing2LimitDegrees=15, twistLimitDegrees=1}
 {
  bone1Name=dyn_chain_down_inn_2_r
  bone2Name=dyn_chain_down_inn_1_r
  swing1LimitDegrees=0
  swing1Motion=Locked
  swing2LimitDegrees=0
  swing2Motion=Locked
  twistLimitDegrees=0
  twistMotion=Locked
 }
 {
  bone1Name=deform_chain_up_inn_l
  bone2Name=spine_05
  swing1LimitDegrees=45
  swing1Motion=Free
  swing2LimitDegrees=45
  swing2Motion=Free
  twistLimitDegrees=45
  twistMotion=Free
 }
 {bone1Name=dyn_chain_up_inn_l, bone2Name=deform_chain_up_inn_l, swing1LimitDegrees=15, swing2LimitDegrees=35, twistLimitDegrees=1}
 {
  bone1Name=dyn_chain_up_inn_1_l
  bone2Name=dyn_chain_up_inn_l
  swing1LimitDegrees=35
  swing1Motion=Locked
  swing2LimitDegrees=35
  swing2Motion=Locked
  twistLimitDegrees=1
  twistMotion=Locked
 }
 {
  bone1Name=deform_chain_up_inn_r
  bone2Name=spine_05
  swing1LimitDegrees=45
  swing1Motion=Free
  swing2LimitDegrees=45
  swing2Motion=Free
  twistLimitDegrees=45
  twistMotion=Free
 }
 {bone1Name=dyn_chain_up_inn_r, bone2Name=deform_chain_up_inn_r, swing1LimitDegrees=15, swing2LimitDegrees=35, twistLimitDegrees=1}
 {
  bone1Name=dyn_chain_up_inn_1_r
  bone2Name=dyn_chain_up_inn_r
  swing1LimitDegrees=5
  swing1Motion=Locked
  swing2LimitDegrees=5
  swing2Motion=Locked
  twistLimitDegrees=1
  twistMotion=Locked
 }
 {
  bone1Name=deform_chain_up_out_l
  bone2Name=spine_05
  swing1LimitDegrees=45
  swing1Motion=Free
  swing2LimitDegrees=45
  swing2Motion=Free
  twistLimitDegrees=45
  twistMotion=Free
 }
 {bone1Name=dyn_chain_up_out_l, bone2Name=deform_chain_up_out_l, swing1LimitDegrees=15, swing2LimitDegrees=35, twistLimitDegrees=1}
 {bone1Name=dyn_chain_down_out_1_l, bone2Name=dyn_chain_up_out_l, swing1LimitDegrees=25, swing2LimitDegrees=25, twistLimitDegrees=1}
 {
  bone1Name=dyn_chain_down_out_2_l
  bone2Name=dyn_chain_down_out_1_l
  swing1LimitDegrees=25
  swing2LimitDegrees=25
  twistLimitDegrees=1
  twistMotion=Locked
 }
 {bone1Name=dyn_chain_up_out_1_l, bone2Name=dyn_chain_up_out_l, swing1LimitDegrees=5, swing2LimitDegrees=5, twistLimitDegrees=1}
 {
  bone1Name=deform_chain_up_out_r
  bone2Name=spine_05
  swing1LimitDegrees=45
  swing1Motion=Free
  swing2LimitDegrees=45
  swing2Motion=Free
  twistLimitDegrees=45
  twistMotion=Free
 }
 {bone1Name=dyn_chain_up_out_r, bone2Name=deform_chain_up_out_r, swing1LimitDegrees=15, swing2LimitDegrees=35, twistLimitDegrees=1}
 {bone1Name=dyn_chain_down_out_1_r, bone2Name=dyn_chain_up_out_r, swing1LimitDegrees=35, swing2LimitDegrees=15, twistLimitDegrees=1}
 {
  bone1Name=dyn_chain_down_out_2_r
  bone2Name=dyn_chain_down_out_1_r
  swing1LimitDegrees=5
  swing2LimitDegrees=5
  twistLimitDegrees=1
  twistMotion=Locked
 }
 {bone1Name=dyn_chain_up_out_1_r, bone2Name=dyn_chain_up_out_r, swing1LimitDegrees=5, swing2LimitDegrees=5, twistLimitDegrees=1}
 {
  bone1Name=spine_05
  bone2Name=spine_04
  swing1LimitDegrees=45
  swing1Motion=Free
  swing2LimitDegrees=45
  swing2Motion=Free
  twistLimitDegrees=45
  twistMotion=Free
 }
 {
  bone1Name=clavicle_l
  bone2Name=spine_05
  swing1LimitDegrees=45
  swing1Motion=Free
  swing2LimitDegrees=45
  swing2Motion=Free
  twistLimitDegrees=45
  twistMotion=Free
 }
 {
  bone1Name=dyn_chain_up_inn_1_r
  bone2Name=dyn_chain_up_out_1_r
  swing1LimitDegrees=45
  swing1Motion=Locked
  swing2LimitDegrees=45
  swing2Motion=Locked
  twistLimitDegrees=45
  twistMotion=Locked
 }
 {
  bone1Name=deform_pec_aim_r
  bone2Name=deform_pec_aim_base_r
  swing1LimitDegrees=45
  swing1Motion=Free
  swing2LimitDegrees=45
  swing2Motion=Free
  twistLimitDegrees=45
  twistMotion=Free
 }
 {
  bone1Name=deform_pec_aim_base_r
  bone2Name=spine_04
  swing1LimitDegrees=45
  swing1Motion=Free
  swing2LimitDegrees=45
  swing2Motion=Free
  twistLimitDegrees=45
  twistMotion=Free
 }
 {
  bone1Name=deform_pec_aim_base_l
  bone2Name=spine_04
  swing1LimitDegrees=45
  swing1Motion=Free
  swing2LimitDegrees=45
  swing2Motion=Free
  twistLimitDegrees=45
  twistMotion=Free
 }
 {
  bone1Name=deform_pec_aim_l
  bone2Name=deform_pec_aim_base_l
  swing1LimitDegrees=45
  swing1Motion=Free
  swing2LimitDegrees=45
  swing2Motion=Free
  twistLimitDegrees=45
  twistMotion=Free
 }
 {
  bone1Name=deform_collar_l
  bone2Name=spine_05
  swing1LimitDegrees=45
  swing1Motion=Free
  swing2LimitDegrees=45
  swing2Motion=Free
  twistLimitDegrees=45
  twistMotion=Free
 }
 {
  bone1Name=dyn_collar_l
  bone2Name=deform_collar_l
  swing1LimitDegrees=30
  swing2LimitDegrees=1
  swing2Motion=Locked
  twistLimitDegrees=1
  twistMotion=Locked
 }
 {
  bone1Name=deform_collar_r
  bone2Name=spine_05
  swing1LimitDegrees=45
  swing1Motion=Free
  swing2LimitDegrees=45
  swing2Motion=Free
  twistLimitDegrees=45
  twistMotion=Free
 }
 {
  bone1Name=dyn_collar_r
  bone2Name=deform_collar_r
  swing1LimitDegrees=30
  swing2LimitDegrees=1
  swing2Motion=Locked
  twistLimitDegrees=1
  twistMotion=Locked
 }
 {
  bone1Name=neck_02
  bone2Name=spine_05
  swing1LimitDegrees=45
  swing1Motion=Free
  swing2LimitDegrees=45
  swing2Motion=Free
  twistLimitDegrees=45
  twistMotion=Free
 }
 {
  bone1Name=head
  bone2Name=neck_02
  swing1LimitDegrees=45
  swing1Motion=Free
  swing2LimitDegrees=45
  swing2Motion=Free
  twistLimitDegrees=45
  twistMotion=Free
 }
 {
  bone1Name=dyn_chain_up_out_l
  bone2Name=dyn_chain_up_inn_l
  swing1LimitDegrees=5
  swing2LimitDegrees=5
  twistLimitDegrees=45
  twistMotion=Free
 }
 {
  bone1Name=dyn_chain_up_inn_1_l
  bone2Name=dyn_chain_up_out_1_l
  swing1LimitDegrees=5
  swing1Motion=Locked
  swing2LimitDegrees=5
  swing2Motion=Locked
  twistLimitDegrees=5
  twistMotion=Locked
 }
 {
  bone1Name=dyn_chain_down_inn_2_l
  bone2Name=dyn_chain_down_out_2_l
  swing1LimitDegrees=45
  swing1Motion=Locked
  swing2LimitDegrees=45
  swing2Motion=Locked
  twistLimitDegrees=45
  twistMotion=Locked
 }
 {
  bone1Name=dyn_chain_down_out_2_r
  bone2Name=dyn_chain_down_inn_2_r
  swing1LimitDegrees=45
  swing1Motion=Locked
  swing2LimitDegrees=45
  swing2Motion=Locked
  twistLimitDegrees=45
  twistMotion=Locked
 }
]
