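# Physics asset of collision shapes for the male medium ButtonCase body, named as the colliders for capes worn on it.
# Capsules cover pelvis, spine, clavicles and neck, with a sphere on the head; arm, leg and foot bodies have no shapes.
# Capsules on the dyn_skirt_b and dyn_skirt_bk bone chains extend collision over the skirt, following those bones.
# It has no constraints, so it serves as a collider set rather than a jointed ragdoll.
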PhysicsAsset /BRCosmetics/Characters/Player/Male/Medium/Bodies/M_MED_ButtonCase/Meshes/M_MED_ButtonCase_CapeColliders
bodies {
 calf_l []
 calf_r []
 clavicle_l [{shapeType=Capsule, length=9.0192, radius=9.8871, center=<9.4266 -1.1806 -2.4977>, rotation=<rot 69.3297 1.4476 176.9462>}]
 clavicle_r [{shapeType=Capsule, length=9.0192, radius=9.8871, center=<-9.4266 1.1806 2.4977>, rotation=<rot 69.3297 1.4476 -3.0538>}]
 dyn_skirt_b_1_l [{shapeType=Capsule, length=15, radius=10, shapeName=dyn_skirt_b_1_l_capsule, center=<20.9243 9.1457 0>, rotation=<rot 0 0 -20.4046>}]
 dyn_skirt_b_1_r [{shapeType=Capsule, length=15, radius=10, shapeName=dyn_skirt_b_1_r_capsule, center=<20.9243 9.1457 0>, rotation=<rot -0.0001 0.0001 18.4192>}]
 dyn_skirt_b_2_l [{
   shapeType=Capsule
   length=9.0632
   radius=14.8765
   shapeName=dyn_skirt_b_2_l_capsule
   center=<22.276 13.8871 0.2632>
   rotation=<rot 0.0254 0.0145 -27.7722>
  }]
 dyn_skirt_b_2_r [{
   shapeType=Capsule
   length=9.3663
   radius=14.8765
   shapeName=dyn_skirt_b_2_r_capsule
   center=<22.276 14.1588 -0.4311>
   rotation=<rot -0.0001 0 23.4398>
  }]
 dyn_skirt_b_l [{shapeType=Capsule, length=23.8308, radius=13.3079, shapeName=dyn_skirt_b_l_capsule, center=<26.6797 10.7761 0>, rotation=<rot 90 0 0>}]
 dyn_skirt_b_r [{shapeType=Capsule, length=23.8308, radius=13.3079, shapeName=dyn_skirt_b_r_capsule, center=<26.6797 10.7761 0>, rotation=<rot 90 0 0>}]
 dyn_skirt_bk_1_l [{shapeType=Capsule, length=14.5909, radius=13.8724, shapeName=dyn_skirt_bk_1_l_capsule, center=<19.9584 10.5013 0>}]
 dyn_skirt_bk_1_r [{shapeType=Capsule, length=14.5909, radius=13.8724, shapeName=dyn_skirt_bk_1_r_capsule, center=<19.9584 10.5013 0>}]
 dyn_skirt_bk_2_l [
  {shapeType=Capsule, length=15.4493, radius=14.8765, shapeName=dyn_skirt_bk_2_l_capsule, center=<22.276 13.6414 0>}
  {shapeType=Capsule, length=15, radius=15, shapeName=dyn_skirt_bk_2_l_capsule, center=<12.524 11.3276 0>, rotation=<rot 90 0 0>}
 ]
 dyn_skirt_bk_2_r [
  {shapeType=Capsule, length=15.4493, radius=14.8765, shapeName=dyn_skirt_bk_2_r_capsule, center=<22.276 13.6414 0>}
  {shapeType=Capsule, length=15, radius=15, shapeName=dyn_skirt_bk_2_r_capsule, center=<12.524 11.3276 0>, rotation=<rot 90 0 0>}
 ]
 dyn_skirt_bk_3_l [{shapeType=Capsule, length=21.3678, radius=15, shapeName=dyn_skirt_bk_3_l_capsule, center=<11.3674 14.2469 0>}]
 dyn_skirt_bk_3_r [{shapeType=Capsule, length=21.3038, radius=15, shapeName=dyn_skirt_bk_3_r_capsule, center=<11.3674 14.2469 0>}]
 dyn_skirt_bk_l [{shapeType=Capsule, length=28.9552, radius=13.3079, shapeName=dyn_skirt_bk_l_capsule, center=<16.5844 8.6365 0>, rotation=<rot 90 0 0>}]
 dyn_skirt_bk_r [{shapeType=Capsule, length=28.9552, radius=13.3079, shapeName=dyn_skirt_bk_r_capsule, center=<16.5844 8.6365 0>, rotation=<rot 90 0 0>}]
 foot_l []
 foot_r []
 head [{shapeType=Sphere, radius=17.3125, center=<11.028 -0.174 0>}]
 lowerarm_l []
 lowerarm_r []
 neck_02 [{shapeType=Capsule, length=13.4375, radius=7.3175, center=<-0.2642 -0.3408 0>, rotation=<rot 90 -171.1193 -175.3978>}]
 pelvis [{shapeType=Capsule, length=8.6456, radius=14.5388, center=<-2.4887 0.1881 0>}]
 spine_01 [{shapeType=Capsule, length=7.9438, radius=13.5586, center=<-1.7965 0.1884 0>}]
 spine_02 [
  {shapeType=Capsule, length=7.3264, radius=13.6322, center=<-1.711 -0.1851 0.0001>}
  {shapeType=Capsule, length=29.3956, radius=11.5807, center=<2.5024 0.4992 3.9825>, rotation=<rot 90 0.0049 0.0049>}
  {shapeType=Capsule, length=29.875, radius=11.1013, center=<2.4704 0.4991 -3.9825>, rotation=<rot 90 -0.0396 -0.0396>}
 ]
 spine_03 [{shapeType=Capsule, length=8.2107, radius=13.7031, center=<0.1445 0.2396 0.0001>}]
 spine_04 [{shapeType=Capsule, length=9.4404, radius=14.1325, center=<2.766 0.7765 0.0002>}]
 spine_05 [
  {shapeType=Capsule, length=8.0226, radius=15.695, center=<-2.2795 1.8109 0>}
  {shapeType=Capsule, length=13.9518, radius=13.9425, center=<-0.2916 0.4452 0>}
 ]
 thigh_l []
 thigh_r []
 upperarm_l []
 upperarm_r []
}
constraints []
